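# Control Rig standard function library holding reusable rigging functions, each in its own graph.
# Constraint, aim, twist, mirror and IK functions (two-bone, spring-smoothed CCDIK, CCDIK plus two-bone leg) drive bones from controls.
# Utilities recolor controls by side, toggle visibility, switch parents, add offset nulls, concatenate names, compute pole-vector transforms and measure bone distances.
# Many functions use Sequence and Branch nodes to gate work and debug drawing behind inputs.
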ControlRigBlueprint /ControlRig/StandardFunctionLibrary/StandardFunctionLibrary
graphs [
 Concat_1_ContainedGraph {nodes [
   n1 = RigVMFunctionEntryNode()
   Concat = /Script/RigVM.RigVMFunction_NameConcat(A=@n1.A, B=@n1.B) {is=struct RigVMUnitNode, template="Concat::Execute(in A,in B,out Result)"}
   n2 = &RigVMFunction_NameConcat(A=@Concat.Result, B=@n1.C) {template="Concat::Execute(in A,in B,out Result)"}
   RigVMFunctionReturnNode(Result=&RigVMFunction_NameConcat(A=&RigVMFunction_NameConcat(A=@n2.Result, B=@n1.D) {template="Concat::Execute(in A,in B,out Result)"}.Result, B=@n1.E) {template="Concat::Execute(in A,in B,out Result)"}.Result)
  ]}
 Concat_1_ContainedGraph {nodes [
   n1 = RigVMFunctionEntryNode()
   Concat = &RigVMFunction_NameConcat(A=@n1.A, B=@n1.B) {template="Concat::Execute(in A,in B,out Result)"}
   n2 = &RigVMFunction_NameConcat(A=@Concat.Result, B=@n1.C) {template="Concat::Execute(in A,in B,out Result)"}
   RigVMFunctionReturnNode(Result=&RigVMFunction_NameConcat(A=&RigVMFunction_NameConcat(A=@n2.Result, B=@n1.D) {template="Concat::Execute(in A,in B,out Result)"}.Result, B=@n1.E) {template="Concat::Execute(in A,in B,out Result)"}.Result)
  ]}
 ContainedGraph {nodes [
   n1 = RigVMFunctionEntryNode()
   RigUnit_SequenceAggregate = &RigVMFunction_Sequence(ExecuteContext=@n1.ExecuteContext)
   BranchNode = &RigVMFunction_ControlFlowBranch(Condition=@n1.Active, ExecuteContext=@RigUnit_SequenceAggregate.A)
   n2 = &RigUnit_GetTransform(bInitial=@n1.Initial, Item=@n1.Item, Space=@n1.Space)
   &RigUnit_SetTransform(
    Weight=1
    ExecutePin=@BranchNode.True
    bInitial=@n1.Initial
    bPropagateToChildren=@n1.Propagate to Children
    Space=@n1.Space
    Item=@n1.Target
    Value=@n2.Transform
   )
   RigVMFunctionReturnNode(Transform=@n2.Transform, ExecuteContext=@RigUnit_SequenceAggregate.B)
  ]}
 ContainedGraph {nodes [
   n1 = RigVMFunctionEntryNode()
   RigUnit_SequenceAggregate_1 = &RigVMFunction_Sequence(ExecuteContext=@n1.ExecuteContext)
   n2 = &RigVMFunction_ControlFlowBranch(Condition=@n1.Debug Draw, ExecuteContext=@RigUnit_SequenceAggregate_1.A)
   n3 = &RigUnit_GetTransform(Item=@n1.object)
   n4 = &RigUnit_GetTransform(Item=@n1.target)
   DebugLineItemSpace = /Script/RigVM.RigVMFunction_DebugLineNoSpace(
    Color=(R=0.5,G=0.5,B=0,A=1)
    DebugDrawSettings=(DepthPriority=SDPG_Foreground,Lifetime=-1)
    Thickness=0.5
    WorldOffset=(Rotation=(X=0,Y=0,Z=0,W=1),Translation=(X=0,Y=0,Z=0),Scale3D=(X=1,Y=1,Z=1))
    bEnabled=True
    ExecutePin=@n2.True
    A=@n3.Transform.Translation
    B=@n4.Transform.Translation
   ) {is=struct RigVMUnitNode}
   n5 = &RigUnit_GetTransform(Item=@n1.Up Vector Object)
   > DebugLineItemSpace_1 = &RigVMFunction_DebugLineNoSpace(
    Color=(R=0.1,G=0.1,B=1,A=0.1)
    DebugDrawSettings=(DepthPriority=SDPG_Foreground,Lifetime=-1)
    Thickness=0.5
    WorldOffset=(Rotation=(X=0,Y=0,Z=0,W=1),Translation=(X=0,Y=0,Z=0),Scale3D=(X=1,Y=1,Z=1))
    bEnabled=True
    A=@n3.Transform.Translation
    B=@n5.Transform.Translation
   )
   RigUnit_SequenceAggregate_2 = &RigVMFunction_Sequence(ExecuteContext=@RigUnit_SequenceAggregate_1.B)
   n6 = &RigVMFunction_ControlFlowBranch(Condition=@n1.Active, ExecuteContext=@RigUnit_SequenceAggregate_2.A)
   n7 = /Script/ControlRig.RigUnit_AimBoneMath(
    DebugSettings=(Scale=10,WorldOffset=(Rotation=(X=0,Y=0,Z=0,W=1),Translation=(X=0,Y=0,Z=0),Scale3D=(X=1,Y=1,Z=1)))
    Primary=(Weight=1,Axis=(X=1,Y=0,Z=0),Target=(X=1,Y=0,Z=0),Kind=Location,Space=(Type=Bone))
    Secondary=(Weight=1,Axis=(X=0,Y=0,Z=1),Target=(X=1,Y=0,Z=0),Kind=Location,Space=(Type=Bone))
    Weight=1
    Primary.Axis=@n1.Axis
    Secondary.Axis=@n1.Up Vector Axis
    InputTransform=@n3.Transform
    Primary.Target=@n4.Transform.Translation
    Secondary.Target=@n5.Transform.Translation
   ) {is=struct RigVMUnitNode}
   &RigUnit_SetTransform(Weight=1, bInitial=False, Value=@n7.Result, ExecutePin=@n6.True, Item=@n1.object)
   RigVMFunctionReturnNode(Result=@n7.Result, ExecuteContext=@RigUnit_SequenceAggregate_2.B)
  ]}
 ContainedGraph {nodes [
   n1 = RigVMFunctionEntryNode()
   RigUnit_SequenceAggregate = &RigVMFunction_Sequence(ExecuteContext=@n1.ExecuteContext)
   n2 = &RigVMFunction_ControlFlowBranch(Condition=@n1.Active, ExecuteContext=@RigUnit_SequenceAggregate.A)
   n3 = RigVMDispatchNode(ExecuteContext=@n2.True, Array=@n1.Items) {template="DISPATCH_RigVMDispatch_ArrayIterator(in Array,out Element,out Index,out Count,out Ratio)"}
   > DebugTransformMutableItemSpace = /Script/RigVM.RigVMFunction_DebugTransformMutableNoSpace(
    Color=(R=1,G=0,B=0,A=1)
    DebugDrawSettings=(DepthPriority=SDPG_Foreground,Lifetime=-1)
    Mode=Axes
    WorldOffset=(Rotation=(X=0,Y=0,Z=0,W=1),Translation=(X=0,Y=0,Z=0),Scale3D=(X=1,Y=1,Z=1))
    bEnabled=True
    Transform=&RigUnit_GetTransform(Item=@n3.Element).Transform
    Thickness=@n1.Thickness
    Scale=@n1.Scale
   ) {is=struct RigVMUnitNode}
   RigVMFunctionReturnNode(ExecuteContext=@RigUnit_SequenceAggregate.B)
  ]}
 ContainedGraph {nodes [
   n1 = RigVMFunctionEntryNode()
   RigUnit_SequenceAggregate = &RigVMFunction_Sequence(ExecuteContext=@n1.ExecuteContext)
   n2 = &RigVMFunction_ControlFlowBranch(Condition=@n1.Active, ExecuteContext=@RigUnit_SequenceAggregate.A)
   DebugTransformMutableItemSpace = &RigVMFunction_DebugTransformMutableNoSpace(
    Color=(R=1,G=0,B=0,A=1)
    DebugDrawSettings=(DepthPriority=SDPG_Foreground,Lifetime=-1)
    Mode=Axes
    WorldOffset=(Rotation=(X=0,Y=0,Z=0,W=1),Translation=(X=0,Y=0,Z=0),Scale3D=(X=1,Y=1,Z=1))
    bEnabled=True
    ExecutePin=@n2.True
    Scale=@n1.Scale
    Thickness=@n1.Thickness
    Transform=&RigUnit_GetTransform(Item=@n1.Item).Transform
   )
   RigVMFunctionReturnNode(ExecuteContext=@RigUnit_SequenceAggregate.B)
  ]}
 ContainedGraph {nodes [
   n1 = RigVMFunctionEntryNode()
   > &RigUnit_AimConstraintLocalSpaceOffset(
    AdvancedSettings=(DebugSettings=(Scale=10,WorldOffset=(Rotation=(X=0,Y=0,Z=0,W=1),Translation=(X=0,Y=0,Z=0),Scale3D=(X=1,Y=1,Z=1))),RotationOrderForFilter=XZY)
    Filter=(bX=True,bY=True,bZ=True)
    Parents=((Item=(Type=Bone),Weight=1))
    Weight=1
    WorldUp=(Target=(X=0,Y=0,Z=1),Kind=Location,Space=(Type=None,Name="None"))
    AimAxis=@n1.Axis
    Parents.0.Item=@n1.End Ctrl
    bMaintainOffset=@n1.Maintain Offset
    Child=@n1.Start Bone
    WorldUp.Space=@n1.Up Ctrl
    UpAxis=@n1.Up Vector Axis
   )
   > &RigUnit_AimConstraintLocalSpaceOffset(
    AdvancedSettings=(DebugSettings=(Scale=10,WorldOffset=(Rotation=(X=0,Y=0,Z=0,W=1),Translation=(X=0,Y=0,Z=0),Scale3D=(X=1,Y=1,Z=1))),RotationOrderForFilter=XZY)
    Filter=(bX=True,bY=True,bZ=True)
    Parents=((Item=(Type=Bone),Weight=1))
    Weight=1
    WorldUp=(Target=(X=0,Y=0,Z=1),Kind=Location,Space=(Type=None,Name="None"))
    AimAxis=@n1.Axis
    Child=@n1.End Bone
    bMaintainOffset=@n1.Maintain Offset
    Parents.0.Item=@n1.Start Ctrl
    UpAxis=@n1.Up Vector Axis
   )
   > /Script/ControlRig.RigUnit_PositionConstraintLocalSpaceOffset(
    Filter=(bX=True,bY=True,bZ=True)
    Parents=((Item=(Type=Bone),Weight=1))
    Weight=1
    bMaintainOffset=True
    Child=@n1.Start Bone
    Parents.0.Item=@n1.Start Ctrl
   ) {is=struct RigVMUnitNode}
   > n2 = &RigUnit_PositionConstraintLocalSpaceOffset(
    Filter=(bX=True,bY=True,bZ=True)
    Parents=((Item=(Type=Bone),Weight=1))
    Weight=1
    bMaintainOffset=True
    Child=@n1.End Bone
    Parents.0.Item=@n1.End Ctrl
   )
   RigUnit_SequenceAggregate = &RigVMFunction_Sequence(ExecuteContext=@n2.ExecutePin)
   n3 = &RigVMFunction_ControlFlowBranch(Condition=@n1.Debug Draw, ExecuteContext=@RigUnit_SequenceAggregate.A)
   DebugTransformMutableItemSpace_1 = &RigVMFunction_DebugTransformMutableNoSpace(
    Color=(R=1,G=0,B=0,A=1)
    DebugDrawSettings=(DepthPriority=SDPG_Foreground,Lifetime=-1)
    Mode=Axes
    Scale=5
    Thickness=0.5
    WorldOffset=(Rotation=(X=0,Y=0,Z=0,W=1),Translation=(X=0,Y=0,Z=0),Scale3D=(X=1,Y=1,Z=1))
    bEnabled=True
    ExecutePin=@n3.True
    Transform=&RigUnit_GetTransform(Item=(Type=Bone)).Transform
   )
   > DebugTransformMutableItemSpace_1_1 = &RigVMFunction_DebugTransformMutableNoSpace(
    Color=(R=1,G=0,B=0,A=1)
    DebugDrawSettings=(DepthPriority=SDPG_Foreground,Lifetime=-1)
    Mode=Axes
    Scale=5
    Thickness=0.5
    WorldOffset=(Rotation=(X=0,Y=0,Z=0,W=1),Translation=(X=0,Y=0,Z=0),Scale3D=(X=1,Y=1,Z=1))
    bEnabled=True
    Transform=&RigUnit_GetTransform(Item=@n1.End Bone).Transform
   )
   RigVMFunctionReturnNode(ExecuteContext=@RigUnit_SequenceAggregate.B)
  ]}
 ContainedGraph {nodes [
   n1 = RigVMFunctionEntryNode()
   > n2 = RigVMDispatchNode(Array=@n1.Bones) {template="DISPATCH_RigVMDispatch_ArrayIterator(in Array,out Element,out Index,out Count,out Ratio)"}
   > &RigUnit_ParentConstraint(
    AdvancedSettings=(RotationOrderForFilter=XZY)
    Parents=((Item=(Type=Bone),Weight=1))
    Weight=1
    Parents.0.Item=RigVMDispatchNode(Index=@n2.Index, Array=@n1.Controls) {template="DISPATCH_RigVMDispatch_ArrayGetAtIndex(in Array,in Index,out Element)"}.Element
    Child=@n2.Element
    Filter.RotationFilter=@n1.Rotation Filter
    Filter.TranslationFilter=@n1.Translation Filter
   )
   RigVMFunctionReturnNode(ExecuteContext=@n2.Completed)
  ]}
 ContainedGraph {nodes [
   n1 = RigVMFunctionEntryNode()
   RigUnit_SequenceAggregate = &RigVMFunction_Sequence(ExecuteContext=@n1.ExecuteContext)
   n2 = &RigVMFunction_ControlFlowBranch(ExecuteContext=@RigUnit_SequenceAggregate.A, Condition=@n1.Draw Transform)
   n3 = &RigUnit_GetTransform(Item=@n1.Bone A, bInitial=@n1.Initial)
   n4 = &RigUnit_GetTransform(Item=@n1.Bone C, bInitial=@n1.Initial)
   n5 = /Script/RigVM.RigVMFunction_MathVectorSub(B=@n3.Transform.Translation, A=@n4.Transform.Translation) {is=struct RigVMUnitNode, template="Subtract::Execute(in A,in B,out Result)"}
   n6 = /Script/RigVM.RigVMFunction_MathFloatDiv(
    A=/Script/RigVM.RigVMFunction_MathVectorDot(A=@n3.Transform.Translation, B=@n4.Transform.Translation) {is=struct RigVMUnitNode, template="Dot,|::Execute(in A,in B,out Result)"}.Result
    B=/Script/RigVM.RigVMFunction_MathVectorLength(Value=@n5.Result) {is=struct RigVMUnitNode, template="Length::Execute(in Value,out Result)"}.Result
   ) {is=struct RigVMUnitNode, template="Divide::Execute(in A,in B,out Result)"}
   n7 = /Script/RigVM.RigVMFunction_MathVectorUnit(Value=@n5.Result) {is=struct RigVMUnitNode, template="Unit::Execute(in Value,out Result)"}
   n8 = &RigUnit_GetTransform(Item=@n1.Bone B, bInitial=@n1.Initial)
   n9 = &RigVMFunction_MathVectorSub(B=@n3.Transform.Translation, A=@n8.Transform.Translation) {template="Subtract::Execute(in A,in B,out Result)"}
   n10 = /Script/RigVM.RigVMFunction_MathVectorScale(
    Factor=0.5
    Value=&RigVMFunction_MathVectorSub(
     B=&RigVMFunction_MathVectorScale(Factor=@n6.Result, Value=@n7.Result) {template="Scale::Execute(in Value,in Factor,out Result)"}.Result
     A=@n9.Result
    ) {template="Subtract::Execute(in A,in B,out Result)"}.Result
   ) {is=struct RigVMUnitNode, template="Scale::Execute(in Value,in Factor,out Result)"}
   n11 = &RigVMFunction_MathVectorUnit(Value=/Script/RigVM.RigVMFunction_MathVectorCross(A=@n5.Result, B=@n9.Result) {is=struct RigVMUnitNode, template="Cross::Execute(in A,in B,out Result)"}.Result) {template="Unit::Execute(in Value,out Result)"}
   n12 = /Script/RigVM.RigVMFunction_MathMatrixFromVectors(
    Origin=@n1.Origin Scale
    X=@n10.Result
    Y=@n11.Result
    Z=&RigVMFunction_MathVectorUnit(Value=&RigVMFunction_MathVectorCross(B=@n10.Result, A=@n11.Result) {template="Cross::Execute(in A,in B,out Result)"}.Result) {template="Unit::Execute(in Value,out Result)"}.Result
   ) {is=struct RigVMUnitNode}
   n13 = &RigVMFunction_MathQuaternionFromEuler(
    RotationOrder=XYZ
    Euler=&RigVMFunction_MathQuaternionToEuler(
     RotationOrder=XYZ
     Value=/Script/RigVM.RigVMFunction_MathMatrixToTransform(Value=@n12.Result) {is=struct RigVMUnitNode, template="Cast::Execute(in Value,out Result)"}.Result.Rotation
    ).Result
   )
   n14 = &RigVMFunction_MathVectorScale(
    Factor=&RigVMFunction_MathFloatDiv(
     A=&RigVMFunction_MathVectorDot(B=@n5.Result, A=@n9.Result) {template="Dot,|::Execute(in A,in B,out Result)"}.Result
     B=&RigVMFunction_MathVectorLength(Value=@n5.Result) {template="Length::Execute(in Value,out Result)"}.Result
    ) {template="Divide::Execute(in A,in B,out Result)"}.Result
    Value=&RigVMFunction_MathVectorUnit(Value=@n5.Result) {template="Unit::Execute(in Value,out Result)"}.Result
   ) {template="Scale::Execute(in Value,in Factor,out Result)"}
   n15 = /Script/RigVM.RigVMFunction_MathVectorAdd(
    B=&RigVMFunction_MathVectorScale(
     Factor=@n1.OffsetFactor
     Value=&RigVMFunction_MathVectorSub(B=@n14.Result, A=@n9.Result) {template="Subtract::Execute(in A,in B,out Result)"}.Result
    ) {template="Scale::Execute(in Value,in Factor,out Result)"}.Result
    A=@n8.Transform.Translation
   ) {is=struct RigVMUnitNode, template="Add::Execute(in A,in B,out Result)"}
   DebugTransformMutableItemSpace = &RigVMFunction_DebugTransformMutableNoSpace(
    Color=(R=1,G=0,B=0,A=1)
    DebugDrawSettings=(DepthPriority=SDPG_Foreground,Lifetime=-1)
    Mode=Axes
    Scale=20
    Thickness=1
    Transform=(Rotation=(X=0,Y=0,Z=0,W=1),Translation=(X=0,Y=0,Z=0),Scale3D=(X=1,Y=1,Z=1))
    WorldOffset=(Rotation=(X=0,Y=0,Z=0,W=1),Translation=(X=0,Y=0,Z=0),Scale3D=(X=1,Y=1,Z=1))
    bEnabled=True
    ExecutePin=@n2.True
    Transform.Rotation=@n13.Result
    Transform.Translation=@n15.Result
   )
   RigVMFunctionReturnNode(
    Transform=(Rotation=(X=0,Y=0,Z=0,W=1),Translation=(X=0,Y=0,Z=0),Scale3D=(X=1,Y=1,Z=1))
    Transform.Rotation=@n13.Result
    Transform.Translation=@n15.Result
    ExecuteContext=@RigUnit_SequenceAggregate.B
   )
  ]}
 ContainedGraph {nodes [
   n1 = RigVMFunctionEntryNode()
   > /Script/ControlRig.RigUnit_CCDIKItemArray(
    BaseRotationLimit=30
    EffectorTransform=(Rotation=(X=0,Y=0,Z=0,W=1),Translation=(X=0,Y=0,Z=0),Scale3D=(X=1,Y=1,Z=1))
    MaxIterations=10
    Precision=1
    RotationLimits=()
    Weight=1
    bPropagateToChildren=True
    bStartFromTail=True
    Items=@n1.Bones
    EffectorTransform.Translation=/Script/ControlRig.RigUnit_SpringInterpVectorV2(
     Current=(X=0,Y=0,Z=0)
     Force=(X=0,Y=0,Z=0)
     TargetVelocityAmount=0
     bInitializeFromTarget=False
     bUseCurrentInput=False
     CriticalDamping=@n1.Critical Damping
     Strength=@n1.Strength
     Target=&RigUnit_GetTransform(Item=@n1.End Control).Transform.Translation
    ) {
     is=struct RigVMUnitNode
     template="SpringInterp::Execute(in Target,in Strength,in CriticalDamping,in Force,in bUseCurrentInput,in Current,in TargetVelocityAmount,in bInitializeFromTarget,out Result,out Velocity)"
    }.Result
   ) {is=struct RigVMUnitNode}
   > RigVMFunctionReturnNode()
  ]}
 ContainedGraph {nodes [
   n1 = RigVMFunctionEntryNode()
   RigUnit_SequenceAggregate = &RigVMFunction_Sequence(ExecuteContext=@n1.ExecuteContext)
   n2 = &RigVMFunction_ControlFlowBranch(Condition=@n1.Negative, ExecuteContext=@RigUnit_SequenceAggregate.A)
   RigVMVariableNode(Value=1, ExecuteContext=@n2.False) {variable=isNegative}
   RigVMVariableNode(Value=-1, ExecuteContext=@n2.True) {variable=isNegative}
   n3 = RigVMDispatchNode(
    Array=RigVMDispatchNode(Array=@n1.controls, Other=@n1.ignore_ctrls) {template="DISPATCH_RigVMDispatch_ArrayDifference(in Array,in Other,out Result)"}.Result
    ExecuteContext=@RigUnit_SequenceAggregate.B
   ) {template="DISPATCH_RigVMDispatch_ArrayIterator(in Array,out Element,out Index,out Count,out Ratio)"}
   RigUnit_SequenceAggregate_1_1 = &RigVMFunction_Sequence(ExecuteContext=@n3.ExecuteContext)
   RigUnit_SequenceAggregate_1 = &RigVMFunction_Sequence(ExecuteContext=@RigUnit_SequenceAggregate_1_1.A)
   BranchNode = &RigVMFunction_ControlFlowBranch(Condition=@n1.X Plane, ExecuteContext=@RigUnit_SequenceAggregate_1.A)
   BranchNode_1 = &RigVMFunction_ControlFlowBranch(ExecuteContext=@BranchNode.False, Condition=@n1.Y Plane)
   BranchNode_2 = &RigVMFunction_ControlFlowBranch(ExecuteContext=@BranchNode_1.False, Condition=@n1.Z Plane)
   RigVMVariableNode(
    ExecuteContext=@BranchNode_2.True
    Value=&RigUnit_GetTransform(Item=(Type=Control), bInitial=true, Item.Name=@n3.Element.Name).Transform.Translation.Z
   ) {variable=position}
   RigVMVariableNode(
    ExecuteContext=@BranchNode_1.True
    Value=&RigUnit_GetTransform(Item=(Type=Control), bInitial=true, Item.Name=@n3.Element.Name).Transform.Translation.Y
   ) {variable=position}
   RigVMVariableNode(
    ExecuteContext=@BranchNode.True
    Value=&RigUnit_GetTransform(Item=(Type=Control), bInitial=true, Item.Name=@n3.Element.Name).Transform.Translation.X
   ) {variable=position}
   n4 = &RigVMFunction_MathFloatMul(B=RigVMVariableNode() {variable=isNegative}.Value, A=RigVMVariableNode() {variable=position}.Value) {template="Multiply::Execute(in A,in B,out Result)"}
   n5 = &RigVMFunction_ControlFlowBranch(
    Condition=/Script/RigVM.RigVMFunction_MathFloatGreater(B=0, A=@n4.Result) {is=struct RigVMUnitNode, template="Greater::Execute(in A,in B,out Result)"}.Result
    ExecuteContext=@RigUnit_SequenceAggregate_1.B
   )
   /Script/ControlRig.RigUnit_SetControlColor(Control=@n3.Element.Name, ExecutePin=@n5.True, Color=@n1.LeftColor) {is=struct RigVMUnitNode, template="SetControlColor::Execute(in Control,in Color)"}
   RigUnit_SequenceAggregate_1_2 = &RigVMFunction_Sequence(ExecuteContext=@RigUnit_SequenceAggregate_1_1.B)
   n6 = &RigVMFunction_ControlFlowBranch(
    Condition=/Script/RigVM.RigVMFunction_MathFloatLess(B=0, A=@n4.Result) {is=struct RigVMUnitNode, template="Less::Execute(in A,in B,out Result)"}.Result
    ExecuteContext=@RigUnit_SequenceAggregate_1_2.A
   )
   &RigUnit_SetControlColor(Control=@n3.Element.Name, ExecutePin=@n6.True, Color=@n1.RightColor) {template="SetControlColor::Execute(in Control,in Color)"}
   MathFloatEquals = RigVMDispatchNode(B=0, A=@n4.Result) {template="DISPATCH_RigVMDispatch_CoreEquals(in A,in B,out Result)"}
   n7 = &RigVMFunction_ControlFlowBranch(
    Condition=&RigVMFunction_MathBoolOr(
     B=@MathFloatEquals.Result
     A=/Script/RigVM.RigVMFunction_MathFloatIsNearlyZero(Tolerance=0.001, Value=@n4.Result) {is=struct RigVMUnitNode, template="IsNearlyZero::Execute(in Value,in Tolerance,out Result)"}.Result
    ).Result
    ExecuteContext=@RigUnit_SequenceAggregate_1_2.B
   )
   &RigUnit_SetControlColor(Control=@n3.Element.Name, ExecutePin=@n7.True, Color=@n1.CenterColor) {template="SetControlColor::Execute(in Control,in Color)"}
   RigVMFunctionReturnNode(ExecuteContext=@n3.Completed)
  ]}
 ContainedGraph {nodes [
   n1 = RigVMFunctionEntryNode()
   SequenceAggregate = &RigVMFunction_Sequence(ExecuteContext=@n1.ExecuteContext)
   n2 = RigVMDispatchNode(ExecuteContext=@SequenceAggregate.A, Array=@n1.controls) {template="DISPATCH_RigVMDispatch_ArrayIterator(in Array,out Element,out Index,out Count,out Ratio)"}
   > SetControlVisibility = /Script/ControlRig.RigUnit_SetControlVisibility(Item=@n2.Element, bVisible=@n1.Active) {is=struct RigVMUnitNode, template="SetControlVisibility::Execute(in Item,in Pattern,in bVisible)"}
   RigVMFunctionReturnNode(ExecuteContext=@SequenceAggregate.B)
  ]}
 ContainedGraph {nodes [
   n1 = RigVMFunctionEntryNode()
   RigUnit_SequenceAggregate = &RigVMFunction_Sequence(ExecuteContext=@n1.ExecuteContext)
   n2 = RigVMDispatchNode(Array=@n1.Bones to Twist, ExecuteContext=@RigUnit_SequenceAggregate.A) {template="DISPATCH_RigVMDispatch_ArrayIterator(in Array,out Element,out Index,out Count,out Ratio)"}
   n3 = RigVMDispatchNode(Array=@n1.Bones to Twist) {template="DISPATCH_RigVMDispatch_ArrayGetNum(in Array,out Num)"}
   n4 = &RigVMFunction_MathFloatDiv(
    A=1
    B=&RigVMFunction_MathFloatAdd(
     B=2
     A=/Script/RigVM.RigVMFunction_MathIntToFloat(Value=@n3.Num) {is=struct RigVMUnitNode, template="Cast::Execute(in Value,out Result)"}.Result
    ) {template="Add::Execute(in A,in B,out Result)"}.Result
   ) {template="Divide::Execute(in A,in B,out Result)"}
   MathIntEquals = RigVMDispatchNode(
    A=@n2.Index
    B=/Script/RigVM.RigVMFunction_MathIntSub(B=1, A=@n2.Count) {is=struct RigVMUnitNode, template="Subtract::Execute(in A,in B,out Result)"}.Result
   ) {template="DISPATCH_RigVMDispatch_CoreEquals(in A,in B,out Result)"}
   n5 = RigVMDispatchNode(
    False=&RigVMFunction_MathFloatAdd(A=@n4.Result, B=&RigVMFunction_MathIntToFloat(Value=@n2.Index) {template="Cast::Execute(in Value,out Result)"}.Result) {template="Add::Execute(in A,in B,out Result)"}.Result
    True=/Script/RigVM.RigVMFunction_MathFloatSub(A=1, B=@n4.Result) {is=struct RigVMUnitNode, template="Subtract::Execute(in A,in B,out Result)"}.Result
    Condition=@MathIntEquals.Result
   ) {template="DISPATCH_RigVMDispatch_If(in Condition,in True,in False,out Result)"}
   > n6 = &RigUnit_RotationConstraintLocalSpaceOffset(
    AdvancedSettings=(InterpolationType=Shortest,RotationOrderForFilter=ZYX)
    Parents=((Item=(Type=Bone,Name="None"),Weight=1),(Item=(Type=Bone,Name="None"),Weight=1))
    Weight=1
    Child=@n2.Element
    Filter=@n1.Axis to Twist
    Parents.1.Item=@n1.End
    Parents.0.Item=@n1.Start
    Parents.0.Weight=@n5.Result
    Parents.1.Weight=&RigVMFunction_MathFloatSub(A=1, B=@n5.Result) {template="Subtract::Execute(in A,in B,out Result)"}.Result
   )
   BranchNode = &RigVMFunction_ControlFlowBranch(Condition=@n1.Debug, ExecuteContext=@n6.ExecutePin)
   DebugTransformMutableItemSpace = &RigVMFunction_DebugTransformMutableNoSpace(
    Color=(R=1,G=0,B=0,A=1)
    DebugDrawSettings=(DepthPriority=SDPG_Foreground,Lifetime=-1)
    Mode=Axes
    Scale=10
    Thickness=0.5
    WorldOffset=(Rotation=(X=0,Y=0,Z=0,W=1),Translation=(X=0,Y=0,Z=0),Scale3D=(X=1,Y=1,Z=1))
    bEnabled=True
    ExecutePin=@BranchNode.True
    Transform=&RigUnit_GetTransform(Item=@n2.Element).Transform
   )
   RigVMFunctionReturnNode(ExecuteContext=@RigUnit_SequenceAggregate.B)
  ]}
 ContainedGraph {nodes [
   n1 = RigVMFunctionEntryNode()
   RigUnit_SequenceAggregate = &RigVMFunction_Sequence(ExecuteContext=@n1.ExecuteContext)
   n2 = &RigVMFunction_MathVectorScale(Value=@n1.Axis, Factor=@n1.Hit Radius) {template="Scale::Execute(in Value,in Factor,out Result)"}
   n3 = &RigUnit_GetTransform(bInitial=false, Item=@n1.Item)
   n4 = /Script/ControlRig.RigUnit_SphereTraceByObjectTypes(
    ObjectTypes=()
    Radius=1
    Start=&RigVMFunction_MathVectorAdd(B=@n2.Result, A=@n3.Transform.Translation) {template="Add::Execute(in A,in B,out Result)"}.Result
    End=&RigVMFunction_MathVectorAdd(
     B=&RigVMFunction_MathVectorScale(Factor=-1, Value=@n2.Result) {template="Scale::Execute(in Value,in Factor,out Result)"}.Result
     A=@n3.Transform.Translation
    ) {template="Add::Execute(in A,in B,out Result)"}.Result
   ) {is=struct RigVMUnitNode}
   n5 = &RigUnit_SphereTraceByObjectTypes(
    ObjectTypes=()
    Radius=1
    End=&RigVMFunction_MathVectorAdd(B=(X=0,Y=25,Z=-50), A=@n3.Transform.Translation) {template="Add::Execute(in A,in B,out Result)"}.Result
    Start=&RigVMFunction_MathVectorAdd(B=(X=0,Y=25,Z=50), A=@n3.Transform.Translation) {template="Add::Execute(in A,in B,out Result)"}.Result
   )
   n6 = &RigUnit_AimBoneMath(
    DebugSettings=(Scale=10,WorldOffset=(Rotation=(X=0,Y=0,Z=0,W=1),Translation=(X=0,Y=0,Z=0),Scale3D=(X=1,Y=1,Z=1)))
    InputTransform=(Rotation=(X=0,Y=0,Z=0,W=1),Translation=(X=0,Y=0,Z=0),Scale3D=(X=1,Y=1,Z=1))
    Primary=(Weight=1,Axis=(X=0,Y=1,Z=0),Target=(X=1,Y=0,Z=0),Kind=Location,Space=(Type=Bone))
    Secondary=(Weight=1,Axis=(X=0,Y=0,Z=1),Target=(X=0,Y=0,Z=0),Kind=Direction,Space=(Type=Bone))
    Weight=1
    InputTransform.Translation=@n4.HitLocation
    Secondary.Target=@n4.HitNormal
    Primary.Target=@n5.HitLocation
   )
   DebugRectangleItemSpace = /Script/RigVM.RigVMFunction_DebugRectangleNoSpace(
    DebugDrawSettings=(DepthPriority=SDPG_Foreground,Lifetime=-1)
    WorldOffset=(Rotation=(X=0,Y=0,Z=0,W=1),Translation=(X=0,Y=0,Z=0),Scale3D=(X=1,Y=1,Z=1))
    bEnabled=True
    Transform=@n6.Result
    Color=@n1.Color
    Scale=@n1.Rectangle Size
    Thickness=@n1.Rectangle Thickness
    ExecutePin=@RigUnit_SequenceAggregate.A
   ) {is=struct RigVMUnitNode}
   > DebugLineItemSpace = &RigVMFunction_DebugLineNoSpace(
    Color=(R=1,G=0,B=1,A=1)
    DebugDrawSettings=(DepthPriority=SDPG_Foreground,Lifetime=-1)
    Thickness=1.5
    WorldOffset=(Rotation=(X=0,Y=0,Z=0,W=1),Translation=(X=0,Y=0,Z=0),Scale3D=(X=1,Y=1,Z=1))
    bEnabled=True
    A=@n4.HitLocation
    B=@n5.HitLocation
   )
   BranchNode = &RigVMFunction_ControlFlowBranch(ExecuteContext=@DebugLineItemSpace.ExecutePin, Condition=@n1.DrawHitPoints)
   DebugTransformMutableItemSpace = &RigVMFunction_DebugTransformMutableNoSpace(
    Color=(R=1,G=0,B=1,A=1)
    DebugDrawSettings=(DepthPriority=SDPG_Foreground,Lifetime=-1)
    Mode=Box
    Scale=5
    Transform=(Rotation=(X=0,Y=0,Z=0,W=1),Translation=(X=0,Y=0,Z=0),Scale3D=(X=1,Y=1,Z=1))
    WorldOffset=(Rotation=(X=0,Y=0,Z=0,W=1),Translation=(X=0,Y=0,Z=0),Scale3D=(X=1,Y=1,Z=1))
    bEnabled=True
    ExecutePin=@BranchNode.True
    Transform.Translation=@n5.HitLocation
   )
   > DebugTransformMutableItemSpace_1 = &RigVMFunction_DebugTransformMutableNoSpace(
    Color=(R=1,G=0,B=1,A=1)
    DebugDrawSettings=(DepthPriority=SDPG_Foreground,Lifetime=-1)
    Mode=Box
    Scale=5
    Transform=(Rotation=(X=0,Y=0,Z=0,W=1),Translation=(X=0,Y=0,Z=0),Scale3D=(X=1,Y=1,Z=1))
    WorldOffset=(Rotation=(X=0,Y=0,Z=0,W=1),Translation=(X=0,Y=0,Z=0),Scale3D=(X=1,Y=1,Z=1))
    bEnabled=True
    Transform.Translation=@n4.HitLocation
   )
   RigVMFunctionReturnNode(ExecuteContext=@RigUnit_SequenceAggregate.B)
  ]}
 ContainedGraph {nodes [
   n1 = RigVMFunctionEntryNode()
   RigUnit_SequenceAggregate = &RigVMFunction_Sequence(ExecuteContext=@n1.ExecuteContext)
   RigVMVariableNode(Value=1000000000, ExecuteContext=@RigUnit_SequenceAggregate.A) {variable=closestDistanceSoFar}
   > n2 = RigVMDispatchNode(Array=@n1.Bones) {template="DISPATCH_RigVMDispatch_ArrayIterator(in Array,out Element,out Index,out Count,out Ratio)"}
   Item = /Script/ControlRig.RigUnit_Item(Item=@n1.Item) {is=struct RigVMUnitNode}
   n3 = &RigVMFunction_MathVectorDistance(
    A=&RigUnit_GetTransform(bInitial=@n1.Initial, Item=@n2.Element).Transform.Translation
    B=&RigUnit_GetTransform(bInitial=@n1.Initial, Item=@Item.Item).Transform.Translation
   )
   n4 = &RigVMFunction_ControlFlowBranch(
    ExecuteContext=@n2.ExecuteContext
    Condition=/Script/RigVM.RigVMFunction_MathFloatLessEqual(A=@n3.Result, B=RigVMVariableNode() {variable=closestDistanceSoFar}.Value) {is=struct RigVMUnitNode, template="LessEqual::Execute(in A,in B,out Result)"}.Result
   )
   RigVMVariableNode(ExecuteContext=@n4.True, Value=@n3.Result) {variable=closestDistanceSoFar}
   > RigVMVariableNode(Value=@n2.Element) {variable=closestBone}
   RigVMFunctionReturnNode(ExecuteContext=@RigUnit_SequenceAggregate.B, Value=RigVMVariableNode() {variable=closestBone}.Value)
  ]}
 ContainedGraph {nodes [
   n1 = RigVMFunctionEntryNode()
   n2 = &RigVMFunction_MathIntSub(B=1, A=RigVMDispatchNode(Array=@n1.Bones) {template="DISPATCH_RigVMDispatch_ArrayGetNum(in Array,out Num)"}.Num) {template="Subtract::Execute(in A,in B,out Result)"}
   > n3 = RigVMVariableNode(Value=@n2.Result) {variable=numBones}
   RigUnit_SequenceAggregate_1 = &RigVMFunction_Sequence(ExecuteContext=@n3.ExecuteContext)
   n4 = RigVMDispatchNode(Array=@n1.Bones, ExecuteContext=@RigUnit_SequenceAggregate_1.A) {template="DISPATCH_RigVMDispatch_ArrayIterator(in Array,out Element,out Index,out Count,out Ratio)"}
   MathIntEquals = RigVMDispatchNode(A=@n4.Index, B=&RigVMFunction_MathIntSub(B=1, A=@n4.Count) {template="Subtract::Execute(in A,in B,out Result)"}.Result) {template="DISPATCH_RigVMDispatch_CoreEquals(in A,in B,out Result)"}
   n5 = &RigVMFunction_ControlFlowBranch(ExecuteContext=@n4.ExecuteContext, Condition=&RigVMFunction_MathBoolNot(Value=@MathIntEquals.Result).Result)
   n6 = &RigUnit_GetTransform(
    Item=RigVMDispatchNode(
     Array=@n1.Bones
     Index=/Script/RigVM.RigVMFunction_MathIntAdd(B=1, A=@n4.Index) {is=struct RigVMUnitNode, template="Add::Execute(in A,in B,out Result)"}.Result
    ) {template="DISPATCH_RigVMDispatch_ArrayGetAtIndex(in Array,in Index,out Element)"}.Element
    bInitial=@n1.Initial
   )
   n7 = &RigUnit_GetTransform(Item=@n4.Element, bInitial=@n1.Initial)
   RigVMVariableNode(
    ExecuteContext=@n5.True
    Value=&RigVMFunction_MathFloatAdd(
     A=&RigVMFunction_MathVectorDistance(B=@n6.Transform.Translation, A=@n7.Transform.Translation).Result
     B=RigVMVariableNode() {variable=currentDistance}.Value
    ) {template="Add::Execute(in A,in B,out Result)"}.Result
   ) {variable=currentDistance}
   RigVMVariableNode(
    ExecuteContext=@n4.Completed
    Value=&RigVMFunction_MathFloatDiv(
     B=&RigVMFunction_MathIntToFloat(Value=@n2.Result) {template="Cast::Execute(in Value,out Result)"}.Result
     A=RigVMVariableNode() {variable=currentDistance}.Value
    ) {template="Divide::Execute(in A,in B,out Result)"}.Result
   ) {variable=averageDistance}
   RigVMFunctionReturnNode(ExecuteContext=@RigUnit_SequenceAggregate_1.B, Length=RigVMVariableNode() {variable=averageDistance}.Value)
   RigUnit_SequenceAggregate = &RigVMFunction_Sequence(ExecuteContext=())
  ]}
 ContainedGraph {nodes [
   n1 = RigVMFunctionEntryNode()
   n2 = &RigUnit_GetTransform(bInitial=true, Item=@n1.Bone A)
   n3 = &RigUnit_GetTransform(bInitial=true, Item=@n1.Bone B)
   n4 = &RigUnit_GetTransform(bInitial=true, Item=@n1.Bone C)
   > /Script/ControlRig.RigUnit_TwoBoneIKSimplePerItem(
    DebugSettings=(Scale=10,WorldOffset=(Rotation=(X=0,Y=0,Z=0,W=1),Translation=(X=0,Y=0,Z=0),Scale3D=(X=1,Y=1,Z=1)))
    PoleVectorKind=Direction
    PoleVectorSpace=(Type=Bone)
    SecondaryAxisWeight=1
    Weight=1
    bPropagateToChildren=True
    ItemA=@n1.Bone A
    ItemB=@n1.Bone B
    EffectorItem=@n1.Bone C
    PrimaryAxis=@n1.Primary Axis
    SecondaryAxis=@n1.Secondary Axis
    ItemALength=&RigVMFunction_MathVectorScale(Factor=&RigVMFunction_MathVectorDistance(A=@n2.Transform.Translation, B=@n3.Transform.Translation).Result, Value=@n1.scale) {template="Scale::Execute(in Value,in Factor,out Result)"}.Result.X
    ItemBLength=&RigVMFunction_MathVectorScale(Factor=&RigVMFunction_MathVectorDistance(A=@n3.Transform.Translation, B=@n4.Transform.Translation).Result, Value=@n1.scale) {template="Scale::Execute(in Value,in Factor,out Result)"}.Result.X
    Effector=/Script/ControlRig.RigUnit_ProjectTransformToNewParent(bChildInitial=True, bOldParentInitial=True, Child=@n1.Bone C, NewParent=@n1.End Ctrl, OldParent=@n1.End Ctrl) {is=struct RigVMUnitNode}.Transform
    PoleVector=&RigUnit_GetTransform(bInitial=false, Item=@n1.Pole Vector Node).Transform.Translation
    StretchMaximumRatio=@n1.Stretch Maximum Ratio
    StretchStartRatio=@n1.Stretch Start Ratio
    bEnableStretch=@n1.Enable Stretch
   ) {is=struct RigVMUnitNode}
   > DebugLineItemSpace = &RigVMFunction_DebugLineNoSpace(
    Color=(R=0.2,G=0.2,B=0.2,A=1)
    DebugDrawSettings=(DepthPriority=SDPG_Foreground,Lifetime=-1)
    WorldOffset=(Rotation=(X=0,Y=0,Z=0,W=1),Translation=(X=0,Y=0,Z=0),Scale3D=(X=1,Y=1,Z=1))
    bEnabled=True
    Thickness=@n1.Line Thickness
    B=&RigUnit_GetTransform(Item=@n1.Bone B).Transform.Translation
    A=&RigUnit_GetTransform(bInitial=false, Item=@n1.Pole Vector Node).Transform.Translation
   )
   > RigVMFunctionReturnNode()
  ]}
 ContainedGraph {nodes [
   n1 = RigVMFunctionEntryNode()
   > n2 = &RigUnit_CCDIKItemArray(
    BaseRotationLimit=30
    MaxIterations=10
    Precision=1
    RotationLimits=()
    Weight=1
    bPropagateToChildren=True
    bStartFromTail=True
    Items=/Script/ControlRig.RigUnit_CollectionChainArray(LastItem=@n1.Bone D, FirstItem=@n1.Bone A) {is=struct RigVMUnitNode}.Items
    EffectorTransform=&RigUnit_GetTransform(bInitial=false, Item=@n1.Foot Ctrl).Transform
   )
   RigUnit_SequenceAggregate = &RigVMFunction_Sequence(ExecuteContext=@n2.ExecutePin)
   n3 = &RigVMFunction_ControlFlowBranch(
    Condition=/Script/ControlRig.RigUnit_ItemExists(Item=@n1.Thigh Ctrl Optional) {is=struct RigVMUnitNode}.Exists
    ExecuteContext=@RigUnit_SequenceAggregate.A
   )
   n4 = /Script/ControlRig.RigUnit_HierarchyGetParent(bDefaultParent=True, Child=@n1.Thigh Ctrl Optional) {is=struct RigVMUnitNode}
   &RigUnit_SetTransform(
    Weight=1
    bInitial=False
    ExecutePin=@n3.True
    Item=@n4.Parent
    Value=&RigUnit_ProjectTransformToNewParent(bChildInitial=True, bOldParentInitial=True, NewParent=@n1.Bone A, OldParent=@n1.Bone A, Child=@n4.Parent).Transform
   )
   n5 = &RigUnit_ProjectTransformToNewParent(bChildInitial=True, bOldParentInitial=True, NewParent=@n1.Thigh Ctrl Optional, OldParent=@n1.Thigh Ctrl Optional, Child=@n1.Bone B)
   > /Script/ControlRig.RigUnit_AimItem(
    DebugSettings=(Scale=10,WorldOffset=(Rotation=(X=0,Y=0,Z=0,W=1),Translation=(X=0,Y=0,Z=0),Scale3D=(X=1,Y=1,Z=1)))
    Primary=(Weight=1,Axis=(X=0,Y=1,Z=0),Target=(X=1,Y=0,Z=0),Kind=Location,Space=(Type=Bone))
    Secondary=(Weight=1,Axis=(X=0,Y=0,Z=1),Target=(X=1,Y=0,Z=0),Kind=Location,Space=(Type=Bone))
    Weight=1
    Primary.Target=@n5.Transform.Translation
    Secondary.Target=@n5.Transform.Translation
    Secondary.Axis=@n1.Secondary Axis
    Primary.Axis=@n1.Primary Axis
    Item=@n1.Bone A
   ) {is=struct RigVMUnitNode}
   IKTwoBone_1 = RigVMFunctionReferenceNode(
    Line Thickness=0.25
    Stretch Start Ratio=0.75
    debug=false
    Bone A=@n1.Bone B
    Bone B=@n1.Bone C
    Bone C=@n1.Bone D
    Enable Stretch=@n1.Enable Stretch
    End Ctrl=@n1.Foot Ctrl
    Pole Vector Node=@n1.Pole Vector Node
    Primary Axis=@n1.Primary Axis
    Secondary Axis=@n1.Secondary Axis
    Stretch Maximum Ratio=@n1.Stretch Maximum Ratio
    scale=@n1.scale
    ExecuteContext=@RigUnit_SequenceAggregate.B
   )
   > RigVMFunctionReturnNode(Pose=(HierarchyTopologyVersion=-1,PoseHash=-1))
  ]}
 ContainedGraph {nodes [
   n1 = RigVMFunctionEntryNode()
   > RigVMDispatchNode(Array=RigVMVariableNode() {variable=transformArray}.Value) {template="DISPATCH_RigVMDispatch_ArrayReset(io Array)"}
   > RigVMDispatchNode(Array=RigVMVariableNode() {variable=posArray}.Value) {template="DISPATCH_RigVMDispatch_ArrayReset(io Array)"}
   > RigVMDispatchNode(Array=RigVMVariableNode() {variable=rotArray}.Value) {template="DISPATCH_RigVMDispatch_ArrayReset(io Array)"}
   > RigVMDispatchNode(Array=RigVMVariableNode() {variable=scaleArray}.Value) {template="DISPATCH_RigVMDispatch_ArrayReset(io Array)"}
   > n2 = RigVMDispatchNode(Array=@n1.Items) {template="DISPATCH_RigVMDispatch_ArrayIterator(in Array,out Element,out Index,out Count,out Ratio)"}
   n3 = &RigUnit_GetTransform(Item=@n2.Element)
   > RigVMDispatchNode(Element=@n3.Transform, Array=RigVMVariableNode() {variable=transformArray}.Value) {template="DISPATCH_RigVMDispatch_ArrayAdd(io Array,in Element,out Index)"}
   > RigVMDispatchNode(Element=@n3.Transform.Rotation, Array=RigVMVariableNode() {variable=rotArray}.Value) {template="DISPATCH_RigVMDispatch_ArrayAdd(io Array,in Element,out Index)"}
   > RigVMDispatchNode(Element=@n3.Transform.Translation, Array=RigVMVariableNode() {variable=posArray}.Value) {template="DISPATCH_RigVMDispatch_ArrayAdd(io Array,in Element,out Index)"}
   > RigVMDispatchNode(Element=@n3.Transform.Scale3D, Array=RigVMVariableNode() {variable=scaleArray}.Value) {template="DISPATCH_RigVMDispatch_ArrayAdd(io Array,in Element,out Index)"}
   RigVMFunctionReturnNode(
    ExecuteContext=@n2.Completed
    scales=RigVMVariableNode() {variable=scaleArray}.Value
    rotations=RigVMVariableNode() {variable=rotArray}.Value
    positions=RigVMVariableNode() {variable=posArray}.Value
    transforms=RigVMVariableNode() {variable=transformArray}.Value
   )
  ]}
 ContainedGraph {nodes [
   n1 = RigVMFunctionEntryNode()
   RigUnit_SequenceAggregate = &RigVMFunction_Sequence(ExecuteContext=@n1.ExecuteContext)
   n2 = &RigVMFunction_ControlFlowBranch(Condition=@n1.Active, ExecuteContext=@RigUnit_SequenceAggregate.A)
   n3 = RigVMDispatchNode(ExecuteContext=@n2.True, Array=@n1.Transforms) {template="DISPATCH_RigVMDispatch_ArrayIterator(in Array,out Element,out Index,out Count,out Ratio)"}
   > DebugTransformMutableItemSpace = &RigVMFunction_DebugTransformMutableNoSpace(
    DebugDrawSettings=(DepthPriority=SDPG_Foreground,Lifetime=-1)
    WorldOffset=(Rotation=(X=0,Y=0,Z=0,W=1),Translation=(X=0,Y=0,Z=0),Scale3D=(X=1,Y=1,Z=1))
    bEnabled=True
    Transform=@n3.Element
    Color=@n1.Color
    Mode=@n1.Mode
    Thickness=@n1.Thickness
    Scale=@n1.Scale
   )
   RigVMFunctionReturnNode(ExecuteContext=@RigUnit_SequenceAggregate.B)
  ]}
 ContainedGraph {nodes [
   n1 = RigVMFunctionEntryNode()
   RigUnit_GetTransform = &RigUnit_GetTransform(bInitial=@n1.Initial, Item=@n1.MirrorFrom)
   n2 = &RigVMFunction_MathTransformMul(
    A=(Rotation=(X=0,Y=0,Z=0,W=1),Translation=(X=0,Y=0,Z=0),Scale3D=(X=1,Y=1,Z=1))
    B=(Rotation=(X=0,Y=0,Z=0,W=1),Translation=(X=0,Y=0,Z=0),Scale3D=(X=-1,Y=1,Z=1))
    B.Scale3D=@n1.MirrorAxis
    A.Rotation=@RigUnit_GetTransform.Transform.Rotation
    A.Translation=@RigUnit_GetTransform.Transform.Translation
   ) {template="Multiply::Execute(in A,in B,out Result)"}
   > &RigUnit_SetTransform(
    Value=(Rotation=(X=0,Y=0,Z=0,W=1),Translation=(X=0,Y=0,Z=0),Scale3D=(X=1,Y=1,Z=1))
    Weight=1
    bInitial=@n1.Initial
    Item=@n1.MirrorTo
    bPropagateToChildren=@n1.Propagate to Children
    Value.Rotation=@n2.Result.Rotation
    Value.Translation=@n2.Result.Translation
    Value.Scale3D=@RigUnit_GetTransform.Transform.Scale3D
   )
   > RigVMFunctionReturnNode(
    Result=(Rotation=(X=0,Y=0,Z=0,W=1),Translation=(X=0,Y=0,Z=0),Scale3D=(X=0,Y=0,Z=0))
    Result.Rotation=@n2.Result.Rotation
    Result.Translation=@n2.Result.Translation
    Result.Scale3D=@RigUnit_GetTransform.Transform.Scale3D
   )
  ]}
 ContainedGraph {nodes [
   n1 = RigVMFunctionEntryNode()
   SequenceAggregate = &RigVMFunction_Sequence(ExecuteContext=@n1.ExecuteContext)
   n2 = &RigVMFunction_ControlFlowBranch(Condition=@n1.IKFK, ExecuteContext=@SequenceAggregate.A)
   n3 = RigVMDispatchNode(ExecuteContext=@n2.False, Array=@n1.Ctrls) {template="DISPATCH_RigVMDispatch_ArrayIterator(in Array,out Element,out Index,out Count,out Ratio)"}
   > /Script/ControlRig.RigUnit_SwitchParent(Mode=DefaultParent, Parent=(Type=Control,Name="None"), bMaintainGlobal=True, Child=@n3.Element) {is=struct RigVMUnitNode}
   n4 = RigVMDispatchNode(ExecuteContext=@n2.True, Array=@n1.Ctrls) {template="DISPATCH_RigVMDispatch_ArrayIterator(in Array,out Element,out Index,out Count,out Ratio)"}
   > &RigUnit_SwitchParent(Mode=ParentItem, bMaintainGlobal=True, Child=@n4.Element, Parent=@n1.IK Parent)
   RigVMFunctionReturnNode(ExecuteContext=@SequenceAggregate.B)
  ]}
 ContainedGraph {nodes [
   n1 = RigVMFunctionEntryNode()
   Replace = /Script/RigVM.RigVMFunction_NameReplace(New=None, Name=@n1.Item.Name, Old=@n1.Replace Suffix) {is=struct RigVMUnitNode, template="Replace::Execute(in Name,in Old,in New,out Result)"}
   Concat = &RigVMFunction_NameConcat(B=@n1.New Suffix, A=@Replace.Result) {template="Concat::Execute(in A,in B,out Result)"}
   > n2 = &RigUnit_HierarchyAddNull(
    Space=GlobalSpace
    Name=@Concat.Result
    Transform=&RigUnit_GetTransform(bInitial=true, Item=@n1.Item).Transform
    Parent=&RigUnit_HierarchyGetParent(bDefaultParent=True, Child=@n1.Item).Parent
   )
   > /Script/ControlRig.RigUnit_SetDefaultParent(Child=@n1.Item, Parent=@n2.Item) {is=struct RigVMUnitNode}
   > RigVMFunctionReturnNode(Null=@n2.Item)
  ]}
 ContainedGraph {nodes [
   n1 = RigVMFunctionEntryNode()
   > n2 = RigVMDispatchNode(Array=@n1.Array) {template="DISPATCH_RigVMDispatch_ArrayIterator(in Array,out Element,out Index,out Count,out Ratio)"}
   LessEqual = /Script/RigVM.RigVMFunction_MathIntLessEqual(A=@n2.Index, B=@n1.Max) {is=struct RigVMUnitNode, template="LessEqual::Execute(in A,in B,out Result)"}
   n3 = &RigVMFunction_ControlFlowBranch(ExecuteContext=@n2.ExecuteContext, Condition=@LessEqual.Result)
   RigVMDispatchNode(Element=@n2.Element, ExecuteContext=@n3.True, Array=RigVMVariableNode() {variable=final_array}.Value) {template="DISPATCH_RigVMDispatch_ArrayAdd(io Array,in Element,out Index)"}
   RigVMFunctionReturnNode(ExecuteContext=@n2.Completed, Value=RigVMVariableNode() {variable=final_array}.Value)
  ]}
 ContainedGraph {nodes [
   n1 = RigVMFunctionEntryNode()
   > SequenceAggregate = RigVMAggregateNode()
   RigVMFunction_Sequence = RigVMAggregateNode(ExecuteContext=@SequenceAggregate.A)
   Concat = &RigVMFunction_NameConcat(B=@n1.Control Suffix, A=@n1.Name) {template="Concat::Execute(in A,in B,out Result)"}
   SpawnControl = /Script/ControlRig.RigUnit_HierarchyAddControlTransform(
    OffsetSpace=LocalSpace
    Parent=(Type=Bone,Name="None")
    Name=@Concat.Result
    OffsetTransform=@n1.Global Position
    Settings=@n1.Settings
    InitialValue=&RigVMFunction_MathTransformMakeRelative(Global=@n1.Global Position, Parent=@n1.Global Position).Local
    ExecutePin=@RigVMFunction_Sequence.A
   ) {
    is=struct RigVMUnitNode
    template="SpawnControl::Execute(in InitialValue,in Settings,in OffsetTransform,in OffsetSpace,in Parent,in Name,out Item)"
   }
   > &RigUnit_SetDefaultParent(Parent=@n1.Parent, Child=@SpawnControl.Item)
   > RigVMVariableNode(Value=@SpawnControl.Item) {variable=the control}
   n2 = &RigVMFunction_ControlFlowBranch(Condition=@n1.Orient to World, ExecuteContext=@RigVMFunction_Sequence.C)
   n3 = RigVMVariableNode() {variable=the control}
   n4 = &RigUnit_GetTransform(bInitial=true, Item=@n3.Value)
   &RigUnit_SetTransform(
    Value=(Rotation=(X=0,Y=0,Z=0,W=1),Translation=(X=0,Y=0,Z=0),Scale3D=(X=1,Y=1,Z=1))
    Weight=1
    bInitial=true
    Value.Scale3D=@n4.Transform.Scale3D
    Value.Translation=@n4.Transform.Translation
    ExecutePin=@n2.True
    Item=@n3.Value
   )
   n5 = &RigVMFunction_ControlFlowBranch(Condition=@n1.Zero Controls, ExecuteContext=@RigVMFunction_Sequence.B)
   &RigUnit_SetTransform(
    Weight=1
    bInitial=true
    Value=&RigUnit_GetTransform(bInitial=true, Item=RigVMVariableNode() {variable=the control}.Value).Transform
    ExecutePin=@n5.True
    Item=RigVMVariableNode() {variable=the control}.Value
   )
   BranchNode_2 = &RigVMFunction_ControlFlowBranch(Condition=@n1.Add Secondary Control, ExecuteContext=@SequenceAggregate.B)
   n6 = &RigUnit_HierarchyAddControlTransform(
    InitialValue=(Rotation=(X=0,Y=0,Z=0,W=1),Translation=(X=0,Y=0,Z=0),Scale3D=(X=1,Y=1,Z=1))
    OffsetSpace=LocalSpace
    OffsetTransform=(Rotation=(X=0,Y=0,Z=0,W=1),Translation=(X=0,Y=0,Z=0),Scale3D=(X=1,Y=1,Z=1))
    ExecutePin=@BranchNode_2.True
    Name=&RigVMFunction_NameConcat(A=&RigVMFunction_NameConcat(B=_sec, A=@n1.Name) {template="Concat::Execute(in A,in B,out Result)"}.Result, B=@n1.Control Suffix) {template="Concat::Execute(in A,in B,out Result)"}.Result
    Settings=@n1.Secondary Settings
    Parent=RigVMVariableNode() {variable=the control}.Value
   ) {template="SpawnControl::Execute(in InitialValue,in Settings,in OffsetTransform,in OffsetSpace,in Parent,in Name,out Item)"}
   > RigVMVariableNode(Value=@n6.Item) {variable=the secondary ctrl}
   n7 = &RigVMFunction_ControlFlowBranch(Condition=@n1.Add Null, ExecuteContext=@SequenceAggregate.C)
   n8 = RigVMVariableNode() {variable=the control}
   Add Null Above = RigVMFunctionReferenceNode(ExecuteContext=@n7.True, Replace Suffix=@n1.Control Suffix, New Suffix=@n1.Null Suffix, Item=@n8.Value)
   > DISPATCH_RigDispatch_SetMetadata = RigVMDispatchNode(Name=Null, NameSpace=Self, Value=@"Add Null Above".Null, Item=@n8.Value) {template="DISPATCH_RigDispatch_SetMetadata(in Item,in Name,in NameSpace,in Value,out Success)"}
   > n9 = RigVMVariableNode(Value=@"Add Null Above".Null) {variable=the null}
   RigVMFunction_ControlFlowBranch = &RigVMFunction_ControlFlowBranch(Condition=@n1.Zero Controls, ExecuteContext=@n9.ExecuteContext)
   &RigUnit_SetTransform(
    Weight=1
    bInitial=true
    Value=&RigUnit_GetTransform(bInitial=true, Item=RigVMVariableNode() {variable=the null}.Value).Transform
    ExecutePin=@RigVMFunction_ControlFlowBranch.True
    Item=RigVMVariableNode() {variable=the control}.Value
   )
   BranchNode_1 = &RigVMFunction_ControlFlowBranch(Condition=@n1.Add Bottom Null, ExecuteContext=@SequenceAggregate.D)
   n10 = &RigUnit_HierarchyAddNull(
    Space=LocalSpace
    Transform=(Rotation=(X=0,Y=0,Z=0,W=1),Translation=(X=0,Y=0,Z=0),Scale3D=(X=1,Y=1,Z=1))
    ExecutePin=@BranchNode_1.True
    Name=&RigVMFunction_NameConcat(B=_ctrl_null, A=@n1.Name) {template="Concat::Execute(in A,in B,out Result)"}.Result
    Parent=RigVMDispatchNode(
     Condition=@n1.Add Secondary Control
     False=RigVMVariableNode() {variable=the control}.Value
     True=RigVMVariableNode() {variable=the secondary ctrl}.Value
    ) {template="DISPATCH_RigVMDispatch_If(in Condition,in True,in False,out Result)"}.Result
   )
   > RigVMDispatchNode(Name=Bottom Null, NameSpace=Self, Value=@n10.Item, Item=RigVMVariableNode() {variable=the control}.Value) {template="DISPATCH_RigDispatch_SetMetadata(in Item,in Name,in NameSpace,in Value,out Success)"}
   > RigVMVariableNode(Value=@n10.Item) {variable=the bottom null}
   RigVMDispatchNode(
    Name=Secondary
    NameSpace=Self
    ExecuteContext=@SequenceAggregate.E
    Item=RigVMVariableNode() {variable=the control}.Value
    Value=RigVMVariableNode() {variable=the secondary ctrl}.Value
   ) {template="DISPATCH_RigDispatch_SetMetadata(in Item,in Name,in NameSpace,in Value,out Success)"}
   > RigVMFunctionReturnNode(
    Control=RigVMVariableNode() {variable=the control}.Value
    Secondary=RigVMVariableNode() {variable=the secondary ctrl}.Value
    Null=RigVMVariableNode() {variable=the null}.Value
    Bottom Null=RigVMVariableNode() {variable=the bottom null}.Value
   )
   n11 = &RigVMFunction_ControlFlowBranch(Condition=@n1.Tag Controls, ExecuteContext=@SequenceAggregate.F)
   /Script/ControlRig.RigUnit_SetMetadataTag(NameSpace=Self, Tag=Controls, ExecutePin=@n11.True, Item=RigVMVariableNode() {variable=the control}.Value) {is=struct RigVMUnitNode}
  ]}
 ContainedGraph {nodes [
   n1 = RigVMFunctionEntryNode()
   SequenceAggregate = &RigVMFunction_Sequence(ExecuteContext=@n1.ExecuteContext)
   SequenceAggregate_1 = &RigVMFunction_Sequence(ExecuteContext=@SequenceAggregate.B)
   SequenceAggregate_2 = &RigVMFunction_Sequence(ExecuteContext=@SequenceAggregate_1.B)
   SequenceAggregate_3 = &RigVMFunction_Sequence(ExecuteContext=@SequenceAggregate_2.B)
   SequenceAggregate_4 = &RigVMFunction_Sequence(ExecuteContext=@SequenceAggregate_3.B)
   RigVMFunctionReturnNode(
    A=@SequenceAggregate.A
    B=@SequenceAggregate_1.A
    C=@SequenceAggregate_2.A
    D=@SequenceAggregate_3.A
    E=@SequenceAggregate_4.A
    F=@SequenceAggregate_4.B
   )
  ]}
 ContainedGraph {nodes [
   n1 = RigVMFunctionEntryNode()
   RigUnit_SequenceAggregate = &RigVMFunction_Sequence(ExecuteContext=@n1.ExecuteContext)
   DISPATCH_RigDispatch_CoreNotEquals = RigVMDispatchNode(B=None, A=@n1.Name) {template="DISPATCH_RigVMDispatch_CoreNotEquals(in A,in B,out Result)"}
   Control Stack at Position = RigVMFunctionReferenceNode(
    Add Bottom Null=@n1.Add Bottom Null
    Add Null=@n1.Add Null
    Add Secondary Control=@n1.Add Secondary Control
    Control Suffix=@n1.Control Suffix
    Null Suffix=@n1.Null Suffix
    Orient to World=@n1.Orient to World
    Parent=@n1.Parent
    Secondary Settings=@n1.Secondary Settings
    Settings=@n1.Settings
    Tag Controls=@n1.Tag Controls
    Zero Controls=@n1.Zero Controls
    Global Position=&RigUnit_GetTransform(bInitial=true, Item=@n1.Item).Transform
    Name=RigVMDispatchNode(Condition=@DISPATCH_RigDispatch_CoreNotEquals.Result, False=@n1.Item.Name, True=@n1.Name) {template="DISPATCH_RigVMDispatch_If(in Condition,in True,in False,out Result)"}.Result
    ExecuteContext=@RigUnit_SequenceAggregate.A
   )
   > RigVMDispatchNode(Name=Item, NameSpace=Self, Item=@"Control Stack at Position".Control, Value=@n1.Item) {template="DISPATCH_RigDispatch_SetMetadata(in Item,in Name,in NameSpace,in Value,out Success)"}
   RigVMFunctionReturnNode(
    Bottom Null=@"Control Stack at Position".Bottom Null
    Control=@"Control Stack at Position".Control
    Null=@"Control Stack at Position".Null
    Secondary=@"Control Stack at Position".Secondary
    ExecuteContext=@RigUnit_SequenceAggregate.B
   )
  ]}
 ContainedGraph {nodes [
   n1 = RigVMFunctionEntryNode()
   > RigVMFunctionReturnNode(Items=/Script/ControlRig.RigUnit_CollectionNameSearchArray(TypeToSearch=Control, PartialName=@n1.Partial Name) {is=struct RigVMUnitNode}.Items)
  ]}
 ContainedGraph {nodes [
   n1 = RigVMFunctionEntryNode()
   RigUnit_SequenceAggregate_1 = &RigVMFunction_Sequence(ExecuteContext=@n1.ExecuteContext)
   n2 = &RigUnit_GetTransform(Item=@n1.Child to Jiggle)
   SpringInterp = &RigUnit_SpringInterpVectorV2(
    Current=(X=0,Y=0,Z=0)
    Force=(X=0,Y=0,Z=0)
    TargetVelocityAmount=0
    bInitializeFromTarget=false
    bUseCurrentInput=false
    CriticalDamping=@n1.Critical Damping
    Strength=@n1.Strength
    Target=&RigUnit_GetTransform(Item=@n1.Ctrl).Transform.Translation
   ) {template="SpringInterp::Execute(in Target,in Strength,in CriticalDamping,in Force,in bUseCurrentInput,in Current,in TargetVelocityAmount,in bInitializeFromTarget,out Result,out Velocity)"}
   RigVMFunctionReturnNode(
    Transform=(Rotation=(X=0,Y=0,Z=0,W=1),Translation=(X=0,Y=0,Z=0),Scale3D=(X=0,Y=0,Z=0))
    Transform.Rotation=@n2.Transform.Rotation
    Transform.Scale3D=@n2.Transform.Scale3D
    Transform.Translation=@SpringInterp.Result
    ExecuteContext=@RigUnit_SequenceAggregate_1.B
   )
   RigUnit_SetTransform = &RigUnit_SetTransform(
    Value=(Rotation=(X=0,Y=-0.7071,Z=0,W=0.7071),Translation=(X=28.7,Y=-184.5,Z=460.4079),Scale3D=(X=1,Y=1,Z=1))
    Weight=1
    bInitial=False
    Item=@n1.Child to Jiggle
    Value.Rotation=@n2.Transform.Rotation
    Value.Scale3D=@n2.Transform.Scale3D
    Value.Translation=@SpringInterp.Result
    ExecutePin=@RigUnit_SequenceAggregate_1.A
   )
  ]}
 ContainedGraph {nodes [
   RigVMFunctionEntryNode()
   > SequenceAggregate_4 = RigVMAggregateNode()
   SequenceAggregate = &RigVMFunction_Sequence(ExecuteContext=@SequenceAggregate_4.A)
   n1 = RigVMVariableNode() {variable=Bones}
   n2 = RigVMDispatchNode(ExecuteContext=@SequenceAggregate.A, Array=@n1.Value) {template="DISPATCH_RigVMDispatch_ArrayIterator(in Array,out Element,out Index,out Count,out Ratio)"}
   Get Next Global Transform = RigVMCollapseNode(Index=@n2.Index, Array=@n1.Value)
   n3 = &RigUnit_GetTransform(bInitial=False, Item=@n2.Element)
   n4 = &RigUnit_AimBoneMath(
    DebugSettings=(bEnabled=False,Scale=10,WorldOffset=(Rotation=(X=0,Y=0,Z=0,W=1),Translation=(X=0,Y=0,Z=0),Scale3D=(X=1,Y=1,Z=1)))
    Primary=(Weight=1,Axis=(X=1,Y=0,Z=0),Target=(X=1,Y=0,Z=0),Kind=Location,Space=(Type=Bone,Name="None"))
    Secondary=(Weight=1,Axis=(X=0,Y=0,Z=1),Target=(X=1,Y=0,Z=0),Kind=Direction,Space=(Type=Bone,Name="None"))
    Weight=1
    Primary.Target=@"Get Next Global Transform".Transform.Translation
    InputTransform=@n3.Transform
    Secondary.Target=/Script/RigVM.RigVMFunction_MathQuaternionRotateVector(Vector=(X=0,Y=0,Z=1), Transform=@n3.Transform.Rotation) {is=struct RigVMUnitNode, template="Rotate Vector::Execute(in Transform,in Vector,out Result)"}.Result
   )
   > RigVMDispatchNode(Name=ChainDynamics:Current, NameSpace=Self, Value=@n4.Result, Item=@n2.Element) {template="DISPATCH_RigDispatch_SetMetadata(in Item,in Name,in NameSpace,in Value,out Success)"}
   n5 = RigVMDispatchNode(ExecuteContext=@SequenceAggregate.B, Array=RigVMVariableNode() {variable=Bones}.Value) {template="DISPATCH_RigVMDispatch_ArrayIterator(in Array,out Element,out Index,out Count,out Ratio)"}
   > &RigUnit_SetTransform(
    Weight=1
    bInitial=False
    Value=RigVMDispatchNode(
     Default=(Rotation=(X=0,Y=0,Z=0,W=1),Translation=(X=0,Y=0,Z=0),Scale3D=(X=1,Y=1,Z=1))
     Name=ChainDynamics:Current
     NameSpace=Self
     Item=@n5.Element
    ) {template="DISPATCH_RigDispatch_GetMetadata(in Item,in Name,in NameSpace,in Default,out Value,out Found)"}.Value
    Item=@n5.Element
   )
   > DISPATCH_RigDispatch_SetMetadata = RigVMDispatchNode(
    Name=ChainDynamics:Blend
    NameSpace=Self
    Item=@n5.Element
    Value=/Script/RigVM.RigVMFunction_MathDoubleLerp(T=@n5.Ratio, A=RigVMVariableNode() {variable=Blend Start}.Value, B=RigVMVariableNode() {variable=Blend End}.Value) {is=struct RigVMUnitNode, template="Interpolate::Execute(in A,in B,in T,out Result)"}.Result
   ) {template="DISPATCH_RigDispatch_SetMetadata(in Item,in Name,in NameSpace,in Value,out Success)"}
   n6 = RigVMDispatchNode(ExecuteContext=@SequenceAggregate_4.B, Array=RigVMVariableNode() {variable=Bones}.Value) {template="DISPATCH_RigVMDispatch_ArrayIterator(in Array,out Element,out Index,out Count,out Ratio)"}
   n7 = &RigUnit_GetTransform(bInitial=False, Item=@n6.Element)
   Get Next Local Transform = RigVMCollapseNode(Index=@n6.Index, Array=RigVMVariableNode() {variable=Bones}.Value)
   DISPATCH_RigDispatch_GetMetadata = RigVMDispatchNode(
    Default=(Rotation=(X=0,Y=0,Z=0,W=1),Translation=(X=0,Y=0,Z=0),Scale3D=(X=1,Y=1,Z=1))
    Name=ChainDynamics:Current
    NameSpace=Self
    Item=@n6.Element
   ) {template="DISPATCH_RigDispatch_GetMetadata(in Item,in Name,in NameSpace,in Default,out Value,out Found)"}
   Aim Spring = RigVMCollapseNode(
    Blend=RigVMDispatchNode(Default=1, Name=ChainDynamics:Blend, NameSpace=Self, Item=@n6.Element) {template="DISPATCH_RigDispatch_GetMetadata(in Item,in Name,in NameSpace,in Default,out Value,out Found)"}.Value
    Global Transform=@n7.Transform
    Aim Target=/Script/RigVM.RigVMFunction_MathTransformTransformVector(
     Location=@"Get Next Local Transform".Transform.Translation
     Transform=/Script/RigVM.RigVMFunction_MathTransformMake(
      Rotation=@DISPATCH_RigDispatch_GetMetadata.Value.Rotation
      Scale=@DISPATCH_RigDispatch_GetMetadata.Value.Scale3D
      Translation=@n7.Transform.Translation
     ) {is=struct RigVMUnitNode}.Result
    ) {is=struct RigVMUnitNode}.Result
    Strength=RigVMVariableNode() {variable=Strength}.Value
    Damping=RigVMVariableNode() {variable=Damping}.Value
    Twist=RigVMVariableNode() {variable=Blend Twist}.Value
   )
   > &RigUnit_SetTransform(Weight=1, bInitial=False, Value=@"Aim Spring".Result, Item=@n6.Element)
   > &RigVMFunction_DebugTransformMutableNoSpace(
    Color=(R=1,G=1,B=0,A=1)
    DebugDrawSettings=(DepthPriority=SDPG_Foreground,Lifetime=-1)
    Mode=Point
    Scale=5
    Thickness=1
    Transform=(Rotation=(X=0,Y=0,Z=0,W=1),Translation=(X=0,Y=0,Z=0),Scale3D=(X=1,Y=1,Z=1))
    WorldOffset=(Rotation=(X=0,Y=0,Z=0,W=1),Translation=(X=0,Y=0,Z=0),Scale3D=(X=1,Y=1,Z=1))
    Transform.Translation=@"Aim Spring".Spring Result
    bEnabled=RigVMVariableNode() {variable=Draw Debug}.Value
   )
   RigVMFunctionReturnNode(ExecuteContext=@SequenceAggregate_4.C)
  ]}
 ContainedGraph {nodes [
   n1 = RigVMFunctionEntryNode()
   SequenceAggregate_4 = &RigVMFunction_Sequence(ExecuteContext=@n1.ExecuteContext)
   SequenceAggregate_4_1 = &RigVMFunction_Sequence(ExecuteContext=@SequenceAggregate_4.B)
   RigVMFunctionReturnNode(A=@SequenceAggregate_4.A, B=@SequenceAggregate_4_1.A, C=@SequenceAggregate_4_1.B)
  ]}
 ContainedGraph {nodes [
   n1 = RigVMFunctionEntryNode()
   Add = &RigVMFunction_MathIntAdd(B=1, A=@n1.Index) {template="Add::Execute(in A,in B,out Result)"}
   DISPATCH_RigVMDispatch_ArrayGetNum = RigVMDispatchNode(Array=@n1.Array) {template="DISPATCH_RigVMDispatch_ArrayGetNum(in Array,out Num)"}
   DISPATCH_RigVMDispatch_CoreEquals = RigVMDispatchNode(
    A=@n1.Index
    B=&RigVMFunction_MathIntSub(B=1, A=@DISPATCH_RigVMDispatch_ArrayGetNum.Num) {template="Subtract::Execute(in A,in B,out Result)"}.Result
   ) {template="DISPATCH_RigVMDispatch_CoreEquals(in A,in B,out Result)"}
   DISPATCH_RigVMDispatch_If = RigVMDispatchNode(False=@Add.Result, Condition=@DISPATCH_RigVMDispatch_CoreEquals.Result, True=@n1.Index) {template="DISPATCH_RigVMDispatch_If(in Condition,in True,in False,out Result)"}
   RigVMFunctionReturnNode(Transform=&RigUnit_GetTransform(
     Space=LocalSpace
     bInitial=False
     Item=RigVMDispatchNode(Index=@DISPATCH_RigVMDispatch_If.Result, Array=@n1.Array) {template="DISPATCH_RigVMDispatch_ArrayGetAtIndex(in Array,in Index,out Element)"}.Element
    ).Transform)
  ]}
 ContainedGraph {nodes [
   n1 = RigVMFunctionEntryNode()
   DISPATCH_RigVMDispatch_ArrayGetNum = RigVMDispatchNode(Array=@n1.Array) {template="DISPATCH_RigVMDispatch_ArrayGetNum(in Array,out Num)"}
   n2 = &RigVMFunction_MathIntSub(B=1, A=@DISPATCH_RigVMDispatch_ArrayGetNum.Num) {template="Subtract::Execute(in A,in B,out Result)"}
   n3 = RigVMDispatchNode(Index=&RigVMFunction_MathIntAdd(B=1, A=@n1.Index) {template="Add::Execute(in A,in B,out Result)"}.Result, Array=@n1.Array) {template="DISPATCH_RigVMDispatch_ArrayGetAtIndex(in Array,in Index,out Element)"}
   DISPATCH_RigVMDispatch_ArrayGetAtIndex = RigVMDispatchNode(Array=@n1.Array, Index=@n1.Index) {template="DISPATCH_RigVMDispatch_ArrayGetAtIndex(in Array,in Index,out Element)"}
   n4 = &RigUnit_GetTransform(bInitial=False, Item=@DISPATCH_RigVMDispatch_ArrayGetAtIndex.Element)
   RigVMFunctionReturnNode(Transform=RigVMDispatchNode(
     Condition=RigVMDispatchNode(A=@n1.Index, B=@n2.Result) {template="DISPATCH_RigVMDispatch_CoreEquals(in A,in B,out Result)"}.Result
     False=&RigUnit_GetTransform(bInitial=False, Item=@n3.Element).Transform
     True=/Script/RigVM.RigVMFunction_MathTransformMakeAbsolute(Local=(Rotation=(X=0,Y=0,Z=0,W=1),Translation=(X=1,Y=0,Z=0),Scale3D=(X=1,Y=1,Z=1)), Parent=@n4.Transform) {is=struct RigVMUnitNode, template="Make Absolute::Execute(in Local,in Parent,out Global)"}.Global
    ) {template="DISPATCH_RigVMDispatch_If(in Condition,in True,in False,out Result)"}.Result)
  ]}
 ContainedGraph {nodes [
   n1 = RigVMFunctionEntryNode()
   n2 = &RigVMFunction_MathDoubleMul(B=RigVMVariableNode() {variable=Blend}.Value, A=RigVMVariableNode() {variable=Twist}.Value)
   n3 = /Script/ControlRig.RigUnit_SpringInterpQuaternionV2(
    Current=(X=0,Y=0,Z=0,W=1)
    TargetVelocityAmount=0
    Torque=(X=0,Y=0,Z=0)
    bInitializeFromTarget=False
    bUseCurrentInput=False
    Target=@n1.Global Transform.Rotation
    CriticalDamping=RigVMVariableNode() {variable=Damping}.Value
    Strength=RigVMVariableNode() {variable=Strength}.Value
   ) {
    is=struct RigVMUnitNode
    template="SpringInterp::Execute(in Target,in Strength,in CriticalDamping,in Torque,in bUseCurrentInput,in Current,in TargetVelocityAmount,in bInitializeFromTarget,out Result,out AngularVelocity)"
   }
   n4 = &RigUnit_SpringInterpVectorV2(
    Current=(X=0,Y=0,Z=0)
    Force=(X=0,Y=0,Z=0)
    TargetVelocityAmount=0
    bInitializeFromTarget=False
    bUseCurrentInput=False
    Strength=RigVMVariableNode() {variable=Strength}.Value
    CriticalDamping=RigVMVariableNode() {variable=Damping}.Value
    Target=RigVMVariableNode() {variable=Aim Target}.Value
   ) {template="SpringInterp::Execute(in Target,in Strength,in CriticalDamping,in Force,in bUseCurrentInput,in Current,in TargetVelocityAmount,in bInitializeFromTarget,out Result,out Velocity)"}
   n5 = &RigUnit_AimBoneMath(
    DebugSettings=(bEnabled=False,Scale=10,WorldOffset=(Rotation=(X=0,Y=0,Z=0,W=1),Translation=(X=0,Y=0,Z=0),Scale3D=(X=1,Y=1,Z=1)))
    InputTransform=(Rotation=(X=0,Y=0,Z=0,W=1),Translation=(X=0,Y=0,Z=0),Scale3D=(X=1,Y=1,Z=1))
    Primary=(Weight=1,Axis=(X=1,Y=0,Z=0),Target=(X=1,Y=0,Z=0),Kind=Location,Space=(Type=Bone,Name="None"))
    Secondary=(Weight=1,Axis=(X=0,Y=0,Z=1),Target=(X=1,Y=0,Z=0),Kind=Direction,Space=(Type=Bone,Name="None"))
    Weight=1
    InputTransform.Translation=@n1.Global Transform.Translation
    Secondary.Target=&RigVMFunction_MathQuaternionRotateVector(
     Vector=(X=0,Y=0,Z=1)
     Transform=/Script/RigVM.RigVMFunction_MathQuaternionSlerp(A=@n1.Global Transform.Rotation, T=@n2.Result, B=@n3.Result) {is=struct RigVMUnitNode, template="Interpolate::Execute(in A,in B,in T,out Result)"}.Result
    ) {template="Rotate Vector::Execute(in Transform,in Vector,out Result)"}.Result
    Primary.Target=@n4.Result
   )
   RigVMFunctionReturnNode(
    Result=/Script/RigVM.RigVMFunction_MathTransformLerp(B=@n5.Result, T=RigVMVariableNode() {variable=Blend}.Value, A=RigVMVariableNode() {variable=Global Transform}.Value) {is=struct RigVMUnitNode, template="Interpolate::Execute(in A,in B,in T,out Result)"}.Result
    Spring Result=@n4.Result
   )
  ]}
 ContainedGraph {nodes [
   n1 = RigVMFunctionEntryNode()
   > n2 = RigVMDispatchNode(Array=@n1.Array) {template="DISPATCH_RigVMDispatch_ArrayIterator(in Array,out Element,out Index,out Count,out Ratio)"}
   RigUnit_SequenceAggregate = &RigVMFunction_Sequence(ExecuteContext=@n2.ExecuteContext)
   n3 = RigVMDispatchNode(Default=(Type=Bone,Name="None"), Name=Null, NameSpace=Self, Item=@n2.Element) {template="DISPATCH_RigDispatch_GetMetadata(in Item,in Name,in NameSpace,in Default,out Value,out Found)"}
   n4 = &RigVMFunction_ControlFlowBranch(Condition=&RigUnit_ItemExists(Item=@n3.Value).Exists, ExecuteContext=@RigUnit_SequenceAggregate.A)
   &RigUnit_SetTransform(
    Weight=1
    bInitial=False
    ExecutePin=@n4.True
    Item=@n3.Value
    Value=&RigUnit_ProjectTransformToNewParent(bChildInitial=True, bOldParentInitial=True, Child=@n3.Value, NewParent=@n1.New Parent, OldParent=@n1.New Parent).Transform
   )
   n5 = RigVMDispatchNode(Default=(Type=Bone,Name="None"), Name=Item, NameSpace=Self, Item=@n2.Element) {template="DISPATCH_RigDispatch_GetMetadata(in Item,in Name,in NameSpace,in Default,out Value,out Found)"}
   &RigUnit_SetTransform(
    Weight=1
    bInitial=False
    Item=@n5.Value
    Value=&RigUnit_ProjectTransformToNewParent(bChildInitial=True, bOldParentInitial=True, NewParent=@n2.Element, OldParent=@n2.Element, Child=@n5.Value).Transform
    ExecutePin=@RigUnit_SequenceAggregate.B
   )
   RigVMFunctionReturnNode(ExecuteContext=@n2.Completed)
  ]}
 ContainedGraph {nodes [
   n1 = RigVMFunctionEntryNode()
   RigUnit_SequenceAggregate = &RigVMFunction_Sequence(ExecuteContext=@n1.ExecuteContext)
   n2 = &RigVMFunction_ControlFlowBranch(Condition=@n1.Active, ExecuteContext=@RigUnit_SequenceAggregate.A)
   n3 = RigVMDispatchNode(ExecuteContext=@n2.True, Array=@n1.Items) {template="DISPATCH_RigVMDispatch_ArrayIterator(in Array,out Element,out Index,out Count,out Ratio)"}
   > DebugTransformMutableItemSpace = &RigVMFunction_DebugTransformMutableNoSpace(
    Color=(R=1,G=0,B=0,A=1)
    DebugDrawSettings=(DepthPriority=SDPG_Foreground,Lifetime=-1)
    Mode=Axes
    WorldOffset=(Rotation=(X=0,Y=0,Z=0,W=1),Translation=(X=0,Y=0,Z=0),Scale3D=(X=1,Y=1,Z=1))
    bEnabled=True
    Transform=&RigUnit_GetTransform(Item=@n3.Element).Transform
    Thickness=@n1.Thickness
    Scale=@n1.Scale
   )
   RigVMFunctionReturnNode(ExecuteContext=@RigUnit_SequenceAggregate.B)
  ]}
 ContainedGraph {nodes [
   n1 = RigVMFunctionEntryNode()
   RigUnit_SequenceAggregate = &RigVMFunction_Sequence(ExecuteContext=@n1.ExecuteContext)
   n2 = &RigVMFunction_ControlFlowBranch(Condition=@n1.Active, ExecuteContext=@RigUnit_SequenceAggregate.A)
   DebugTransformMutableItemSpace = &RigVMFunction_DebugTransformMutableNoSpace(
    Color=(R=1,G=0,B=0,A=1)
    DebugDrawSettings=(DepthPriority=SDPG_Foreground,Lifetime=-1)
    Mode=Axes
    WorldOffset=(Rotation=(X=0,Y=0,Z=0,W=1),Translation=(X=0,Y=0,Z=0),Scale3D=(X=1,Y=1,Z=1))
    bEnabled=True
    ExecutePin=@n2.True
    Scale=@n1.Scale
    Thickness=@n1.Thickness
    Transform=&RigUnit_GetTransform(Item=@n1.Item).Transform
   )
   RigVMFunctionReturnNode(ExecuteContext=@RigUnit_SequenceAggregate.B)
  ]}
 ContainedGraph {nodes [
   n1 = RigVMFunctionEntryNode()
   RigUnit_SequenceAggregate = &RigVMFunction_Sequence(ExecuteContext=@n1.ExecuteContext)
   n2 = &RigVMFunction_ControlFlowBranch(Condition=@n1.Active, ExecuteContext=@RigUnit_SequenceAggregate.A)
   n3 = RigVMDispatchNode(ExecuteContext=@n2.True, Array=@n1.Transforms) {template="DISPATCH_RigVMDispatch_ArrayIterator(in Array,out Element,out Index,out Count,out Ratio)"}
   > DebugTransformMutableItemSpace = &RigVMFunction_DebugTransformMutableNoSpace(
    DebugDrawSettings=(DepthPriority=SDPG_Foreground,Lifetime=-1)
    WorldOffset=(Rotation=(X=0,Y=0,Z=0,W=1),Translation=(X=0,Y=0,Z=0),Scale3D=(X=1,Y=1,Z=1))
    bEnabled=True
    Transform=@n3.Element
    Color=@n1.Color
    Mode=@n1.Mode
    Thickness=@n1.Thickness
    Scale=@n1.Scale
   )
   RigVMFunctionReturnNode(ExecuteContext=@RigUnit_SequenceAggregate.B)
  ]}
 ContainedGraph {nodes [
   n1 = RigVMFunctionEntryNode()
   RigUnit_SequenceAggregate = &RigVMFunction_Sequence(ExecuteContext=@n1.ExecuteContext)
   n2 = &RigVMFunction_MathVectorScale(Value=@n1.Axis, Factor=@n1.Hit Radius) {template="Scale::Execute(in Value,in Factor,out Result)"}
   n3 = &RigUnit_GetTransform(bInitial=false, Item=@n1.Item)
   n4 = &RigUnit_SphereTraceByObjectTypes(
    ObjectTypes=()
    Radius=1
    Start=&RigVMFunction_MathVectorAdd(B=@n2.Result, A=@n3.Transform.Translation) {template="Add::Execute(in A,in B,out Result)"}.Result
    End=&RigVMFunction_MathVectorAdd(
     B=&RigVMFunction_MathVectorScale(Factor=-1, Value=@n2.Result) {template="Scale::Execute(in Value,in Factor,out Result)"}.Result
     A=@n3.Transform.Translation
    ) {template="Add::Execute(in A,in B,out Result)"}.Result
   )
   n5 = &RigUnit_SphereTraceByObjectTypes(
    ObjectTypes=()
    Radius=1
    End=&RigVMFunction_MathVectorAdd(B=(X=0,Y=25,Z=-50), A=@n3.Transform.Translation) {template="Add::Execute(in A,in B,out Result)"}.Result
    Start=&RigVMFunction_MathVectorAdd(B=(X=0,Y=25,Z=50), A=@n3.Transform.Translation) {template="Add::Execute(in A,in B,out Result)"}.Result
   )
   n6 = &RigUnit_AimBoneMath(
    DebugSettings=(Scale=10,WorldOffset=(Rotation=(X=0,Y=0,Z=0,W=1),Translation=(X=0,Y=0,Z=0),Scale3D=(X=1,Y=1,Z=1)))
    InputTransform=(Rotation=(X=0,Y=0,Z=0,W=1),Translation=(X=0,Y=0,Z=0),Scale3D=(X=1,Y=1,Z=1))
    Primary=(Weight=1,Axis=(X=0,Y=1,Z=0),Target=(X=1,Y=0,Z=0),Kind=Location,Space=(Type=Bone))
    Secondary=(Weight=1,Axis=(X=0,Y=0,Z=1),Target=(X=0,Y=0,Z=0),Kind=Direction,Space=(Type=Bone))
    Weight=1
    InputTransform.Translation=@n4.HitLocation
    Secondary.Target=@n4.HitNormal
    Primary.Target=@n5.HitLocation
   )
   DebugRectangleItemSpace = &RigVMFunction_DebugRectangleNoSpace(
    DebugDrawSettings=(DepthPriority=SDPG_Foreground,Lifetime=-1)
    WorldOffset=(Rotation=(X=0,Y=0,Z=0,W=1),Translation=(X=0,Y=0,Z=0),Scale3D=(X=1,Y=1,Z=1))
    bEnabled=True
    Transform=@n6.Result
    Color=@n1.Color
    Scale=@n1.Rectangle Size
    Thickness=@n1.Rectangle Thickness
    ExecutePin=@RigUnit_SequenceAggregate.A
   )
   > DebugLineItemSpace = &RigVMFunction_DebugLineNoSpace(
    Color=(R=1,G=0,B=1,A=1)
    DebugDrawSettings=(DepthPriority=SDPG_Foreground,Lifetime=-1)
    Thickness=1.5
    WorldOffset=(Rotation=(X=0,Y=0,Z=0,W=1),Translation=(X=0,Y=0,Z=0),Scale3D=(X=1,Y=1,Z=1))
    bEnabled=True
    A=@n4.HitLocation
    B=@n5.HitLocation
   )
   BranchNode = &RigVMFunction_ControlFlowBranch(ExecuteContext=@DebugLineItemSpace.ExecutePin, Condition=@n1.DrawHitPoints)
   DebugTransformMutableItemSpace = &RigVMFunction_DebugTransformMutableNoSpace(
    Color=(R=1,G=0,B=1,A=1)
    DebugDrawSettings=(DepthPriority=SDPG_Foreground,Lifetime=-1)
    Mode=Box
    Scale=5
    Transform=(Rotation=(X=0,Y=0,Z=0,W=1),Translation=(X=0,Y=0,Z=0),Scale3D=(X=1,Y=1,Z=1))
    WorldOffset=(Rotation=(X=0,Y=0,Z=0,W=1),Translation=(X=0,Y=0,Z=0),Scale3D=(X=1,Y=1,Z=1))
    bEnabled=True
    ExecutePin=@BranchNode.True
    Transform.Translation=@n5.HitLocation
   )
   > DebugTransformMutableItemSpace_1 = &RigVMFunction_DebugTransformMutableNoSpace(
    Color=(R=1,G=0,B=1,A=1)
    DebugDrawSettings=(DepthPriority=SDPG_Foreground,Lifetime=-1)
    Mode=Box
    Scale=5
    Transform=(Rotation=(X=0,Y=0,Z=0,W=1),Translation=(X=0,Y=0,Z=0),Scale3D=(X=1,Y=1,Z=1))
    WorldOffset=(Rotation=(X=0,Y=0,Z=0,W=1),Translation=(X=0,Y=0,Z=0),Scale3D=(X=1,Y=1,Z=1))
    bEnabled=True
    Transform.Translation=@n4.HitLocation
   )
   RigVMFunctionReturnNode(ExecuteContext=@RigUnit_SequenceAggregate.B)
  ]}
 ContainedGraph {nodes [
   n1 = RigVMFunctionEntryNode()
   > n2 = &RigUnit_CCDIKItemArray(
    BaseRotationLimit=30
    MaxIterations=10
    Precision=1
    RotationLimits=()
    Weight=1
    bPropagateToChildren=True
    bStartFromTail=True
    Items=&RigUnit_CollectionChainArray(LastItem=@n1.Bone D, FirstItem=@n1.Bone A).Items
    EffectorTransform=&RigUnit_GetTransform(bInitial=false, Item=@n1.Foot Ctrl).Transform
   )
   RigUnit_SequenceAggregate = &RigVMFunction_Sequence(ExecuteContext=@n2.ExecutePin)
   n3 = &RigVMFunction_ControlFlowBranch(Condition=&RigUnit_ItemExists(Item=@n1.Thigh Ctrl Optional).Exists, ExecuteContext=@RigUnit_SequenceAggregate.A)
   n4 = &RigUnit_HierarchyGetParent(bDefaultParent=True, Child=@n1.Thigh Ctrl Optional)
   &RigUnit_SetTransform(
    Weight=1
    bInitial=False
    ExecutePin=@n3.True
    Item=@n4.Parent
    Value=&RigUnit_ProjectTransformToNewParent(bChildInitial=True, bOldParentInitial=True, NewParent=@n1.Bone A, OldParent=@n1.Bone A, Child=@n4.Parent).Transform
   )
   n5 = &RigUnit_ProjectTransformToNewParent(bChildInitial=True, bOldParentInitial=True, NewParent=@n1.Thigh Ctrl Optional, OldParent=@n1.Thigh Ctrl Optional, Child=@n1.Bone B)
   > &RigUnit_AimItem(
    DebugSettings=(scale=10,WorldOffset=(Rotation=(X=0,Y=0,Z=0,W=1),Translation=(X=0,Y=0,Z=0),Scale3D=(X=1,Y=1,Z=1)))
    Primary=(Weight=1,Axis=(X=0,Y=1,Z=0),Target=(X=1,Y=0,Z=0),Kind=Location,Space=(Type=Bone))
    Secondary=(Weight=1,Axis=(X=0,Y=0,Z=1),Target=(X=1,Y=0,Z=0),Kind=Location,Space=(Type=Bone))
    Weight=1
    Primary.Target=@n5.Transform.Translation
    Secondary.Target=@n5.Transform.Translation
    Secondary.Axis=@n1.Secondary Axis
    Primary.Axis=@n1.Primary Axis
    Item=@n1.Bone A
   )
   IKTwoBone_1 = RigVMFunctionReferenceNode(
    Line Thickness=0.25
    Stretch Start Ratio=0.75
    debug=false
    Bone A=@n1.Bone B
    Bone B=@n1.Bone C
    Bone C=@n1.Bone D
    Enable Stretch=@n1.Enable Stretch
    End Ctrl=@n1.Foot Ctrl
    Pole Vector Node=@n1.Pole Vector Node
    Primary Axis=@n1.Primary Axis
    Secondary Axis=@n1.Secondary Axis
    Stretch Maximum Ratio=@n1.Stretch Maximum Ratio
    scale=@n1.scale
    ExecuteContext=@RigUnit_SequenceAggregate.B
   )
   > RigVMFunctionReturnNode(Pose=(HierarchyTopologyVersion=-1,PoseHash=-1))
  ]}
 ContainedGraph {nodes [
   RigVMFunctionEntryNode()
   n1 = RigVMVariableNode() {variable=Bones}
   > n2 = RigVMVariableNode(Value=&RigVMFunction_MathIntSub(B=1, A=RigVMDispatchNode(Array=@n1.Value) {template="DISPATCH_RigVMDispatch_ArrayGetNum(in Array,out Num)"}.Num) {template="Subtract::Execute(in A,in B,out Result)"}.Result) {variable=numBones}
   RigUnit_SequenceAggregate_1 = &RigVMFunction_Sequence(ExecuteContext=@n2.ExecuteContext)
   n3 = RigVMDispatchNode(ExecuteContext=@RigUnit_SequenceAggregate_1.A, Array=RigVMVariableNode() {variable=Bones}.Value) {template="DISPATCH_RigVMDispatch_ArrayIterator(in Array,out Element,out Index,out Count,out Ratio)"}
   > RigVMVariableNode(Value=@n3.Element) {variable=currentElement}
   > n4 = RigVMVariableNode(Value=@n3.Index) {variable=currentIndex}
   n5 = RigVMVariableNode() {variable=currentIndex}
   MathIntEquals = RigVMDispatchNode(B=&RigVMFunction_MathIntSub(B=1, A=@n5.Value) {template="Subtract::Execute(in A,in B,out Result)"}.Result, A=@n5.Value) {template="DISPATCH_RigVMDispatch_CoreEquals(in A,in B,out Result)"}
   n6 = &RigVMFunction_ControlFlowBranch(Condition=&RigVMFunction_MathBoolNot(Value=@MathIntEquals.Result).Result, ExecuteContext=@n4.ExecuteContext)
   n7 = RigVMVariableNode() {variable=currentIndex}
   n8 = &RigUnit_GetTransform(
    Item=RigVMDispatchNode(
     Index=&RigVMFunction_MathIntAdd(B=1, A=@n7.Value) {template="Add::Execute(in A,in B,out Result)"}.Result
     Array=RigVMVariableNode() {variable=Bones}.Value
    ) {template="DISPATCH_RigVMDispatch_ArrayGetAtIndex(in Array,in Index,out Element)"}.Element
    bInitial=RigVMVariableNode() {variable=Initial}.Value
   )
   n9 = &RigUnit_GetTransform(Item=RigVMVariableNode() {variable=currentElement}.Value, bInitial=RigVMVariableNode() {variable=Initial}.Value)
   RigVMVariableNode(
    ExecuteContext=@n6.True
    Value=&RigVMFunction_MathFloatAdd(
     A=&RigVMFunction_MathVectorDistance(B=@n8.Transform.Translation, A=@n9.Transform.Translation).Result
     B=RigVMVariableNode() {variable=currentDistance}.Value
    ) {template="Add::Execute(in A,in B,out Result)"}.Result
   ) {variable=currentDistance}
   RigVMFunctionReturnNode(ExecuteContext=@RigUnit_SequenceAggregate_1.B, Length=RigVMVariableNode() {variable=currentDistance}.Value)
  ]}
 ContainedGraph {nodes [
   RigVMFunctionEntryNode()
   > RigVMDispatchNode(Array=RigVMVariableNode() {variable=Current Transforms}.Value) {template="DISPATCH_RigVMDispatch_ArrayReset(io Array)"}
   > RigVMDispatchNode(Array=RigVMVariableNode() {variable=Blend Values}.Value) {template="DISPATCH_RigVMDispatch_ArrayReset(io Array)"}
   > SequenceAggregate_4 = RigVMAggregateNode()
   SequenceAggregate = &RigVMFunction_Sequence(ExecuteContext=@SequenceAggregate_4.A)
   n1 = RigVMVariableNode() {variable=Bones}
   n2 = RigVMDispatchNode(ExecuteContext=@SequenceAggregate.A, Array=@n1.Value) {template="DISPATCH_RigVMDispatch_ArrayIterator(in Array,out Element,out Index,out Count,out Ratio)"}
   Get Next Global Transform = RigVMCollapseNode(Index=@n2.Index, Primary Axis=RigVMVariableNode() {variable=Primary Axis}.Value, Array=@n1.Value)
   n3 = &RigUnit_GetTransform(bInitial=False, Item=@n2.Element)
   n4 = &RigUnit_AimBoneMath(
    DebugSettings=(bEnabled=False,Scale=10,WorldOffset=(Rotation=(X=0,Y=0,Z=0,W=1),Translation=(X=0,Y=0,Z=0),Scale3D=(X=1,Y=1,Z=1)))
    Primary=(Weight=1,Axis=(X=1,Y=0,Z=0),Target=(X=1,Y=0,Z=0),Kind=Location,Space=(Type=Bone,Name="None"))
    Secondary=(Weight=1,Axis=(X=0,Y=0,Z=1),Target=(X=1,Y=0,Z=0),Kind=Direction,Space=(Type=Bone,Name="None"))
    Weight=1
    Primary.Target=@"Get Next Global Transform".Transform.Translation
    InputTransform=@n3.Transform
    Secondary.Target=&RigVMFunction_MathQuaternionRotateVector(Transform=@n3.Transform.Rotation, Vector=RigVMVariableNode() {variable=Secondary Axis}.Value) {template="Rotate Vector::Execute(in Transform,in Vector,out Result)"}.Result
    Primary.Axis=RigVMVariableNode() {variable=Primary Axis}.Value
    Secondary.Axis=RigVMVariableNode() {variable=Secondary Axis}.Value
   )
   > Add = RigVMDispatchNode(Element=@n4.Result, Array=RigVMVariableNode() {variable=Current Transforms}.Value) {template="DISPATCH_RigVMDispatch_ArrayAdd(io Array,in Element,out Index)"}
   n5 = RigVMDispatchNode(ExecuteContext=@SequenceAggregate.B, Array=RigVMVariableNode() {variable=Bones}.Value) {template="DISPATCH_RigVMDispatch_ArrayIterator(in Array,out Element,out Index,out Count,out Ratio)"}
   > &RigUnit_SetTransform(
    Weight=1
    bInitial=False
    Value=RigVMDispatchNode(Index=@n5.Index, Array=RigVMVariableNode() {variable=Current Transforms}.Value) {template="DISPATCH_RigVMDispatch_ArrayGetAtIndex(in Array,in Index,out Element)"}.Element
    Item=@n5.Element
   )
   > RigVMDispatchNode(
    Element=&RigVMFunction_MathDoubleLerp(T=@n5.Ratio, A=RigVMVariableNode() {variable=Blend Start}.Value, B=RigVMVariableNode() {variable=Blend End}.Value) {template="Interpolate::Execute(in A,in B,in T,out Result)"}.Result
    Array=RigVMVariableNode() {variable=Blend Values}.Value
   ) {template="DISPATCH_RigVMDispatch_ArrayAdd(io Array,in Element,out Index)"}
   n6 = RigVMDispatchNode(ExecuteContext=@SequenceAggregate_4.B, Array=RigVMVariableNode() {variable=Bones}.Value) {template="DISPATCH_RigVMDispatch_ArrayIterator(in Array,out Element,out Index,out Count,out Ratio)"}
   n7 = &RigUnit_GetTransform(bInitial=False, Item=@n6.Element)
   Get Next Local Transform = RigVMCollapseNode(Index=@n6.Index, Array=RigVMVariableNode() {variable=Bones}.Value)
   n8 = RigVMDispatchNode(Index=@n6.Index, Array=RigVMVariableNode() {variable=Current Transforms}.Value) {template="DISPATCH_RigVMDispatch_ArrayGetAtIndex(in Array,in Index,out Element)"}
   Aim Spring = RigVMCollapseNode(
    Blend=RigVMDispatchNode(Index=@n6.Index, Array=RigVMVariableNode() {variable=Blend Values}.Value) {template="DISPATCH_RigVMDispatch_ArrayGetAtIndex(in Array,in Index,out Element)"}.Element
    Global Transform=@n7.Transform
    Aim Target=&RigVMFunction_MathTransformTransformVector(
     Location=@"Get Next Local Transform".Transform.Translation
     Transform=&RigVMFunction_MathTransformMake(Rotation=@n8.Element.Rotation, Scale=@n8.Element.Scale3D, Translation=@n7.Transform.Translation).Result
    ).Result
    Primary Axis=RigVMVariableNode() {variable=Primary Axis}.Value
    Secondary Axis=RigVMVariableNode() {variable=Secondary Axis}.Value
    Force=RigVMVariableNode() {variable=Force}.Value
    Strength=RigVMVariableNode() {variable=Strength}.Value
    Damping=RigVMVariableNode() {variable=Damping}.Value
    Twist=RigVMVariableNode() {variable=Blend Twist}.Value
   )
   > &RigUnit_SetTransform(Weight=1, bInitial=False, Value=@"Aim Spring".Result, Item=@n6.Element)
   > &RigVMFunction_DebugTransformMutableNoSpace(
    Color=(R=1,G=1,B=0,A=1)
    DebugDrawSettings=(DepthPriority=SDPG_Foreground,Lifetime=-1)
    Mode=Point
    Scale=5
    Thickness=1
    Transform=(Rotation=(X=0,Y=0,Z=0,W=1),Translation=(X=0,Y=0,Z=0),Scale3D=(X=1,Y=1,Z=1))
    WorldOffset=(Rotation=(X=0,Y=0,Z=0,W=1),Translation=(X=0,Y=0,Z=0),Scale3D=(X=1,Y=1,Z=1))
    Transform.Translation=@"Aim Spring".Spring Result
    bEnabled=RigVMVariableNode() {variable=Draw Debug}.Value
   )
   RigVMFunctionReturnNode(ExecuteContext=@SequenceAggregate_4.C)
  ]}
 ContainedGraph {nodes [
   n1 = RigVMFunctionEntryNode()
   SequenceAggregate_4 = &RigVMFunction_Sequence(ExecuteContext=@n1.ExecuteContext)
   SequenceAggregate_4_1 = &RigVMFunction_Sequence(ExecuteContext=@SequenceAggregate_4.B)
   RigVMFunctionReturnNode(A=@SequenceAggregate_4.A, B=@SequenceAggregate_4_1.A, C=@SequenceAggregate_4_1.B)
  ]}
 ContainedGraph {nodes [
   n1 = RigVMFunctionEntryNode()
   Add = &RigVMFunction_MathIntAdd(B=1, A=@n1.Index) {template="Add::Execute(in A,in B,out Result)"}
   DISPATCH_RigVMDispatch_ArrayGetNum = RigVMDispatchNode(Array=@n1.Array) {template="DISPATCH_RigVMDispatch_ArrayGetNum(in Array,out Num)"}
   DISPATCH_RigVMDispatch_CoreEquals = RigVMDispatchNode(
    A=@n1.Index
    B=&RigVMFunction_MathIntSub(B=1, A=@DISPATCH_RigVMDispatch_ArrayGetNum.Num) {template="Subtract::Execute(in A,in B,out Result)"}.Result
   ) {template="DISPATCH_RigVMDispatch_CoreEquals(in A,in B,out Result)"}
   DISPATCH_RigVMDispatch_If = RigVMDispatchNode(False=@Add.Result, Condition=@DISPATCH_RigVMDispatch_CoreEquals.Result, True=@n1.Index) {template="DISPATCH_RigVMDispatch_If(in Condition,in True,in False,out Result)"}
   RigVMFunctionReturnNode(Transform=&RigUnit_GetTransform(
     Space=LocalSpace
     bInitial=False
     Item=RigVMDispatchNode(Index=@DISPATCH_RigVMDispatch_If.Result, Array=@n1.Array) {template="DISPATCH_RigVMDispatch_ArrayGetAtIndex(in Array,in Index,out Element)"}.Element
    ).Transform)
  ]}
 ContainedGraph {nodes [
   n1 = RigVMFunctionEntryNode()
   DISPATCH_RigVMDispatch_ArrayGetNum = RigVMDispatchNode(Array=@n1.Array) {template="DISPATCH_RigVMDispatch_ArrayGetNum(in Array,out Num)"}
   n2 = &RigVMFunction_MathIntSub(B=1, A=@DISPATCH_RigVMDispatch_ArrayGetNum.Num) {template="Subtract::Execute(in A,in B,out Result)"}
   n3 = RigVMDispatchNode(Index=&RigVMFunction_MathIntAdd(B=1, A=@n1.Index) {template="Add::Execute(in A,in B,out Result)"}.Result, Array=@n1.Array) {template="DISPATCH_RigVMDispatch_ArrayGetAtIndex(in Array,in Index,out Element)"}
   DISPATCH_RigVMDispatch_ArrayGetAtIndex = RigVMDispatchNode(Array=@n1.Array, Index=@n1.Index) {template="DISPATCH_RigVMDispatch_ArrayGetAtIndex(in Array,in Index,out Element)"}
   n4 = &RigUnit_GetTransform(bInitial=False, Item=@DISPATCH_RigVMDispatch_ArrayGetAtIndex.Element)
   RigVMFunctionReturnNode(Transform=RigVMDispatchNode(
     Condition=RigVMDispatchNode(A=@n1.Index, B=@n2.Result) {template="DISPATCH_RigVMDispatch_CoreEquals(in A,in B,out Result)"}.Result
     False=&RigUnit_GetTransform(bInitial=False, Item=@n3.Element).Transform
     True=&RigVMFunction_MathTransformMakeAbsolute(Local=(Rotation=(X=0,Y=0,Z=0,W=1),Translation=(X=1,Y=0,Z=0),Scale3D=(X=1,Y=1,Z=1)), Local.Translation=@n1.Primary Axis, Parent=@n4.Transform) {template="Make Absolute::Execute(in Local,in Parent,out Global)"}.Global
    ) {template="DISPATCH_RigVMDispatch_If(in Condition,in True,in False,out Result)"}.Result)
  ]}
 ContainedGraph {nodes [
   n1 = RigVMFunctionEntryNode()
   n2 = &RigVMFunction_MathDoubleMul(B=RigVMVariableNode() {variable=Blend}.Value, A=RigVMVariableNode() {variable=Twist}.Value)
   n3 = &RigUnit_SpringInterpQuaternionV2(
    Current=(X=0,Y=0,Z=0,W=1)
    TargetVelocityAmount=0
    Torque=(X=0,Y=0,Z=0)
    bInitializeFromTarget=False
    bUseCurrentInput=False
    Target=@n1.Global Transform.Rotation
    CriticalDamping=RigVMVariableNode() {variable=Damping}.Value
    Strength=RigVMVariableNode() {variable=Strength}.Value
   ) {template="SpringInterp::Execute(in Target,in Strength,in CriticalDamping,in Torque,in bUseCurrentInput,in Current,in TargetVelocityAmount,in bInitializeFromTarget,out Result,out AngularVelocity)"}
   n4 = &RigUnit_SpringInterpVectorV2(
    Current=(X=0,Y=0,Z=0)
    TargetVelocityAmount=0
    bInitializeFromTarget=False
    bUseCurrentInput=False
    Force=@n1.Force
    Strength=RigVMVariableNode() {variable=Strength}.Value
    CriticalDamping=RigVMVariableNode() {variable=Damping}.Value
    Target=RigVMVariableNode() {variable=Aim Target}.Value
   ) {template="SpringInterp::Execute(in Target,in Strength,in CriticalDamping,in Force,in bUseCurrentInput,in Current,in TargetVelocityAmount,in bInitializeFromTarget,out Result,out Velocity)"}
   n5 = &RigUnit_AimBoneMath(
    DebugSettings=(bEnabled=False,Scale=10,WorldOffset=(Rotation=(X=0,Y=0,Z=0,W=1),Translation=(X=0,Y=0,Z=0),Scale3D=(X=1,Y=1,Z=1)))
    InputTransform=(Rotation=(X=0,Y=0,Z=0,W=1),Translation=(X=0,Y=0,Z=0),Scale3D=(X=1,Y=1,Z=1))
    Primary=(Weight=1,Axis=(X=1,Y=0,Z=0),Target=(X=1,Y=0,Z=0),Kind=Location,Space=(Type=Bone,Name="None"))
    Secondary=(Weight=1,Axis=(X=0,Y=0,Z=1),Target=(X=1,Y=0,Z=0),Kind=Direction,Space=(Type=Bone,Name="None"))
    Weight=1
    InputTransform.Translation=@n1.Global Transform.Translation
    Primary.Axis=@n1.Primary Axis
    Secondary.Target=&RigVMFunction_MathQuaternionRotateVector(
     Transform=&RigVMFunction_MathQuaternionSlerp(A=@n1.Global Transform.Rotation, T=@n2.Result, B=@n3.Result) {template="Interpolate::Execute(in A,in B,in T,out Result)"}.Result
     Vector=@n1.Secondary Axis
    ) {template="Rotate Vector::Execute(in Transform,in Vector,out Result)"}.Result
    Primary.Target=@n4.Result
    Secondary.Axis=@n1.Secondary Axis
   )
   RigVMFunctionReturnNode(
    Result=&RigVMFunction_MathTransformLerp(B=@n5.Result, T=RigVMVariableNode() {variable=Blend}.Value, A=RigVMVariableNode() {variable=Global Transform}.Value) {template="Interpolate::Execute(in A,in B,in T,out Result)"}.Result
    Spring Result=@n4.Result
   )
  ]}
 Multiply_1_1_ContainedGraph {nodes [
   n1 = RigVMFunctionEntryNode()
   RigVMFunctionReturnNode(Result=/Script/RigVM.RigVMFunction_MathVectorMul(B=@n1.C, A=&RigVMFunction_MathVectorMul(A=@n1.A, B=@n1.B) {template="Multiply::Execute(in A,in B,out Result)"}.Result) {is=struct RigVMUnitNode, template="Multiply::Execute(in A,in B,out Result)"}.Result)
  ]}
 New Function_ContainedGraph {nodes [
   n1 = RigVMFunctionEntryNode()
   > &RigUnit_SetTransform(
    Weight=1
    bInitial=False
    Item=@n1.Child
    Value=&RigUnit_ProjectTransformToNewParent(bChildInitial=True, bNewParentInitial=False, bOldParentInitial=True, Child=@n1.Child, NewParent=@n1.New Parent, OldParent=@n1.New Parent).Transform
   )
   > RigVMFunctionReturnNode()
  ]}
 New Function_ContainedGraph {nodes [
   n1 = RigVMFunctionEntryNode()
   > n2 = RigVMVariableNode(Value=false) {variable=found}
   RigVMFunction_ControlFlowBranch = &RigVMFunction_ControlFlowBranch(Condition=@n1.Children, ExecuteContext=@n2.ExecuteContext)
   RigVMFunctionReturnNode(ExecuteContext=@RigVMFunction_ControlFlowBranch.Completed, Item Found=RigVMVariableNode() {variable=found item}.Value)
   DISPATCH_RigVMDispatch_ArrayIterator = RigVMDispatchNode(
    Array=/Script/ControlRig.RigUnit_HierarchyGetParentsItemArray(bDefaultParent=True, bIncludeChild=true, bReverse=false, Child=@n1.Item) {is=struct RigVMUnitNode}.Parents
    ExecuteContext=@RigVMFunction_ControlFlowBranch.False
   ) {template="DISPATCH_RigVMDispatch_ArrayIterator(in Array,out Element,out Index,out Count,out Ratio)"}
   > n3 = RigVMDispatchNode(Array=@n1.Search for) {template="DISPATCH_RigVMDispatch_ArrayIterator(in Array,out Element,out Index,out Count,out Ratio)"}
   Contains = /Script/RigVM.RigVMFunction_Contains(Name=@DISPATCH_RigVMDispatch_ArrayIterator.Element.Name, Search=@n3.Element) {is=struct RigVMUnitNode, template="Contains::Execute(in Name,in Search,out Result)"}
   RigVMFunction_MathBoolNot = &RigVMFunction_MathBoolNot(Value=RigVMVariableNode() {variable=found}.Value)
   RigVMFunction_MathBoolAnd = &RigVMFunction_MathBoolAnd(A=@Contains.Result, B=@RigVMFunction_MathBoolNot.Result)
   n4 = &RigVMFunction_ControlFlowBranch(ExecuteContext=@n3.ExecuteContext, Condition=@RigVMFunction_MathBoolAnd.Result)
   RigVMVariableNode(Value=true, ExecuteContext=@n4.True) {variable=found}
   > RigVMVariableNode(Value=@DISPATCH_RigVMDispatch_ArrayIterator.Element) {variable=found item}
   RigVMVariableNode(Value=False, ExecuteContext=@RigVMFunction_ControlFlowBranch.True) {variable=bad name}
   > n5 = RigVMDispatchNode(Array=/Script/ControlRig.RigUnit_CollectionChildrenArray(bDefaultChildren=True, bIncludeParent=true, bRecursive=true, Parent=@n1.Item, TypeToSearch=@n1.Type to Search) {is=struct RigVMUnitNode}.Items) {template="DISPATCH_RigVMDispatch_ArrayIterator(in Array,out Element,out Index,out Count,out Ratio)"}
   > RigVMVariableNode(Value=false) {variable=bad name}
   > n6 = RigVMDispatchNode(Array=@n1.Ignore) {template="DISPATCH_RigVMDispatch_ArrayIterator(in Array,out Element,out Index,out Count,out Ratio)"}
   n7 = &RigVMFunction_ControlFlowBranch(
    Condition=&RigVMFunction_Contains(Name=@n5.Element.Name, Search=@n6.Element) {template="Contains::Execute(in Name,in Search,out Result)"}.Result
    ExecuteContext=@n6.ExecuteContext
   )
   RigVMVariableNode(Value=true, ExecuteContext=@n7.True) {variable=bad name}
   n8 = RigVMDispatchNode(ExecuteContext=@n6.Completed, Array=@n1.Search for) {template="DISPATCH_RigVMDispatch_ArrayIterator(in Array,out Element,out Index,out Count,out Ratio)"}
   RigVMFunction_MathBoolAnd_1 = RigVMAggregateNode(
    A=&RigVMFunction_Contains(Name=@n5.Element.Name, Search=@n8.Element) {template="Contains::Execute(in Name,in Search,out Result)"}.Result
    B=&RigVMFunction_MathBoolNot(Value=RigVMVariableNode() {variable=found}.Value).Result
    C=&RigVMFunction_MathBoolNot(Value=RigVMVariableNode() {variable=bad name}.Value).Result
   )
   n9 = &RigVMFunction_ControlFlowBranch(ExecuteContext=@n8.ExecuteContext, Condition=@RigVMFunction_MathBoolAnd_1.Result)
   RigVMVariableNode(Value=true, ExecuteContext=@n9.True) {variable=found}
   > RigVMVariableNode(Value=@n5.Element) {variable=found item}
  ]}
 New Function_ContainedGraph {nodes [
   n1 = RigVMFunctionEntryNode()
   > DISPATCH_RigVMDispatch_ArrayIterator = RigVMDispatchNode(Array=@n1.Controls) {template="DISPATCH_RigVMDispatch_ArrayIterator(in Array,out Element,out Index,out Count,out Ratio)"}
   n2 = /Script/ControlRig.RigUnit_GetShapeTransform(Control=@DISPATCH_RigVMDispatch_ArrayIterator.Element.Name) {is=struct RigVMUnitNode}
   Multiply_1 = RigVMAggregateNode(
    B=(X=1,Y=1,Z=1)
    C=(X=0,Y=0,Z=0)
    B.X=@n1.Global Scale
    B.Y=@n1.Global Scale
    B.Z=@n1.Global Scale
    C.X=@n1.Local Scale
    C.Y=@n1.Local Scale
    C.Z=@n1.Local Scale
    A=@n2.Transform.Scale3D
   )
   > /Script/ControlRig.RigUnit_SetShapeTransform(
    Transform=(Rotation=(X=0,Y=0,Z=0,W=1),Translation=(X=0,Y=0,Z=0),Scale3D=(X=1,Y=1,Z=1))
    Control=@DISPATCH_RigVMDispatch_ArrayIterator.Element.Name
    Transform.Rotation=@n2.Transform.Rotation
    Transform.Translation=@n2.Transform.Translation
    Transform.Scale3D=@Multiply_1.Result
   ) {is=struct RigVMUnitNode}
   RigVMFunctionReturnNode(ExecuteContext=@DISPATCH_RigVMDispatch_ArrayIterator.Completed)
  ]}
 New Function_ContainedGraph {nodes [
   n1 = RigVMFunctionEntryNode()
   > &RigUnit_SetTransform(Weight=1, bInitial=true, Item=@n1.Control, Value=&RigUnit_GetTransform(bInitial=true, Item=@n1.Control).Transform)
   > RigVMFunctionReturnNode()
  ]}
 New Function_ContainedGraph {nodes [
   n1 = RigVMFunctionEntryNode()
   Replace = &RigVMFunction_NameReplace(New=None, Name=@n1.Item.Name, Old=@n1.Replace Suffix) {template="Replace::Execute(in Name,in Old,in New,out Result)"}
   Concat = &RigVMFunction_NameConcat(B=@n1.New Suffix, A=@Replace.Result) {template="Concat::Execute(in A,in B,out Result)"}
   DISPATCH_RigVMDispatch_If = RigVMDispatchNode(True=@n1.Transform, Condition=@n1.Use Transform, False=&RigUnit_GetTransform(bInitial=true, Item=@n1.Item).Transform) {template="DISPATCH_RigVMDispatch_If(in Condition,in True,in False,out Result)"}
   > n2 = &RigUnit_HierarchyAddNull(Space=GlobalSpace, Name=@Concat.Result, Transform=@DISPATCH_RigVMDispatch_If.Result, Parent=@n1.Item)
   > RigVMFunctionReturnNode(Null=@n2.Item)
  ]}
 New Function_ContainedGraph {nodes [
   n1 = RigVMFunctionEntryNode()
   RigVMFunction_Sequence = &RigVMFunction_Sequence(ExecuteContext=@n1.ExecuteContext)
   n2 = /Script/ControlRig.RigUnit_GetCandidates() {is=struct RigVMUnitNode}
   n3 = &RigVMFunction_ControlFlowBranch(
    Condition=RigVMDispatchNode(B=@n1.Connector, A=@n2.Connector) {template="DISPATCH_RigVMDispatch_CoreEquals(in A,in B,out Result)"}.Result
    ExecuteContext=@RigVMFunction_Sequence.A
   )
   n4 = RigVMDispatchNode(Array=@n2.Candidates, ExecuteContext=@n3.True) {template="DISPATCH_RigVMDispatch_ArrayIterator(in Array,out Element,out Index,out Count,out Ratio)"}
   DISPATCH_RigDispatch_GetModuleMetadata = RigVMDispatchNode(Default=(Type=Bone,Name="None"), Name=@n1.Module Metadata Name, NameSpace=@n1.Module Metadata NameSpace) {template="DISPATCH_RigDispatch_GetModuleMetadata(in Name,in NameSpace,in Default,out Value,out Found)"}
   > n5 = RigVMVariableNode(Value=@DISPATCH_RigDispatch_GetModuleMetadata.Found) {variable=found_metadata}
   RigVMFunction_ControlFlowBranch = &RigVMFunction_ControlFlowBranch(Condition=@DISPATCH_RigDispatch_GetModuleMetadata.Found, ExecuteContext=@n5.ExecuteContext)
   DISPATCH_RigVMDispatch_CoreEquals = RigVMDispatchNode(B=@DISPATCH_RigDispatch_GetModuleMetadata.Value, A=@n4.Element) {template="DISPATCH_RigVMDispatch_CoreEquals(in A,in B,out Result)"}
   n6 = &RigVMFunction_ControlFlowBranch(Condition=@DISPATCH_RigVMDispatch_CoreEquals.Result, ExecuteContext=@RigVMFunction_ControlFlowBranch.True)
   /Script/ControlRig.RigUnit_SetDefaultMatch(Default=@n4.Element, ExecutePin=@n6.True) {is=struct RigVMUnitNode, template="Set Default Match::Execute(in Default)"}
   RigVMFunctionReturnNode(ExecuteContext=@RigVMFunction_Sequence.B, Found=RigVMVariableNode() {variable=found_metadata}.Value)
  ]}
 New Function_ContainedGraph {nodes [
   n1 = RigVMFunctionEntryNode()
   DISPATCH_RigDispatch_GetModuleMetadata = RigVMDispatchNode(Default=(R=1,G=1,B=1,A=1), Name=Color Center, NameSpace=Root) {template="DISPATCH_RigDispatch_GetModuleMetadata(in Name,in NameSpace,in Default,out Value,out Found)"}
   RigVMFunction_MathBoolNot = &RigVMFunction_MathBoolNot(Value=@n1.Override)
   DISPATCH_RigVMDispatch_If = RigVMDispatchNode(True=@DISPATCH_RigDispatch_GetModuleMetadata.Value, False=@n1.CenterColor, Condition=@RigVMFunction_MathBoolNot.Result) {template="DISPATCH_RigVMDispatch_If(in Condition,in True,in False,out Result)"}
   Get_Module_Metadata_4_1 = RigVMDispatchNode(Default=(X=1,Y=0,Z=0), Name=Mirror Axis, NameSpace=Root) {template="DISPATCH_RigDispatch_GetModuleMetadata(in Name,in NameSpace,in Default,out Value,out Found)"}
   Less = /Script/RigVM.RigVMFunction_MathDoubleLess(B=0, A=@Get_Module_Metadata_4_1.Value.X) {is=struct RigVMUnitNode, template="Less::Execute(in A,in B,out Result)"}
   n2 = &RigVMFunction_MathDoubleLess(B=0, A=@Get_Module_Metadata_4_1.Value.Y) {template="Less::Execute(in A,in B,out Result)"}
   n3 = &RigVMFunction_MathDoubleLess(B=0, A=@Get_Module_Metadata_4_1.Value.Z) {template="Less::Execute(in A,in B,out Result)"}
   RigVMFunction_MathBoolOr_3 = RigVMAggregateNode(A=@Less.Result, B=@n2.Result, C=@n3.Result)
   Greater = /Script/RigVM.RigVMFunction_MathDoubleGreater(B=0, A=@Get_Module_Metadata_4_1.Value.X) {is=struct RigVMUnitNode, template="Greater::Execute(in A,in B,out Result)"}
   RigVMFunction_MathBoolOr = &RigVMFunction_MathBoolOr(A=@Greater.Result, B=@Less.Result)
   > ColorizeControls = RigVMFunctionReferenceNode(
    ignore_ctrls=()
    CenterColor=@DISPATCH_RigVMDispatch_If.Result
    LeftColor=RigVMDispatchNode(
     True=RigVMDispatchNode(Default=(R=1,G=1,B=1,A=1), Name=Color left, NameSpace=Root) {template="DISPATCH_RigDispatch_GetModuleMetadata(in Name,in NameSpace,in Default,out Value,out Found)"}.Value
     False=@n1.LeftColor
     Condition=@RigVMFunction_MathBoolNot.Result
    ) {template="DISPATCH_RigVMDispatch_If(in Condition,in True,in False,out Result)"}.Result
    RightColor=RigVMDispatchNode(
     True=RigVMDispatchNode(Default=(R=1,G=1,B=1,A=1), Name=Color Right, NameSpace=Root) {template="DISPATCH_RigDispatch_GetModuleMetadata(in Name,in NameSpace,in Default,out Value,out Found)"}.Value
     False=@n1.RightColor
     Condition=@RigVMFunction_MathBoolNot.Result
    ) {template="DISPATCH_RigVMDispatch_If(in Condition,in True,in False,out Result)"}.Result
    controls=@n1.Controls
    Negative=RigVMDispatchNode(False=false, True=true, Condition=@RigVMFunction_MathBoolOr_3.Result) {template="DISPATCH_RigVMDispatch_If(in Condition,in True,in False,out Result)"}.Result
    X Plane=@RigVMFunction_MathBoolOr.Result
    Y Plane=&RigVMFunction_MathBoolOr(
     A=&RigVMFunction_MathDoubleGreater(B=0, A=@Get_Module_Metadata_4_1.Value.Y) {template="Greater::Execute(in A,in B,out Result)"}.Result
     B=@n2.Result
    ).Result
    Z Plane=&RigVMFunction_MathBoolOr(
     A=&RigVMFunction_MathDoubleGreater(B=0, A=@Get_Module_Metadata_4_1.Value.Z) {template="Greater::Execute(in A,in B,out Result)"}.Result
     B=@n3.Result
    ).Result
   )
   > RigVMFunctionReturnNode()
   Get_Module_Metadata_4 = RigVMDispatchNode(Default=(X=1,Y=0,Z=0), Name=Mirror Axis, NameSpace=Root) {template="DISPATCH_RigDispatch_GetModuleMetadata(in Name,in NameSpace,in Default,out Value,out Found)"}
   Axis to Bool = RigVMFunctionReferenceNode(Value=@Get_Module_Metadata_4.Value)
   Axis to Bool_1 = RigVMFunctionReferenceNode(Value=(X=0,Y=0,Z=0))
  ]}
 New Function_ContainedGraph {nodes [
   n1 = RigVMFunctionEntryNode()
   > RigVMFunction_Sequence = RigVMAggregateNode()
   RigVMFunction_Sequence_1 = RigVMAggregateNode(ExecuteContext=@RigVMFunction_Sequence.A)
   DISPATCH_RigVMDispatch_ArrayIterator = RigVMDispatchNode(Array=@n1.Bones, ExecuteContext=@RigVMFunction_Sequence_1.A) {template="DISPATCH_RigVMDispatch_ArrayIterator(in Array,out Element,out Index,out Count,out Ratio)"}
   > DISPATCH_RigDispatch_SetMetadata = RigVMDispatchNode(
    Name=Orig Parent
    NameSpace=Self
    Item=@DISPATCH_RigVMDispatch_ArrayIterator.Element
    Value=&RigUnit_HierarchyGetParent(bDefaultParent=True, Child=@DISPATCH_RigVMDispatch_ArrayIterator.Element).Parent
   ) {template="DISPATCH_RigDispatch_SetMetadata(in Item,in Name,in NameSpace,in Value,out Success)"}
   n2 = RigVMVariableNode() {variable=Bones}
   > &RigUnit_SetDefaultParent(
    Child=@DISPATCH_RigVMDispatch_ArrayIterator.Element
    Parent=&RigUnit_HierarchyGetParent(
     bDefaultParent=True
     Child=RigVMDispatchNode(Index=0, Array=@n2.Value) {template="DISPATCH_RigVMDispatch_ArrayGetAtIndex(in Array,in Index,out Element)"}.Element
    ).Parent
   )
   DISPATCH_RigVMDispatch_ArrayGetNum = RigVMDispatchNode(Array=@n1.Bones) {template="DISPATCH_RigVMDispatch_ArrayGetNum(in Array,out Num)"}
   Interpolate = &RigVMFunction_MathTransformLerp(
    T=0.5
    A=&RigUnit_GetTransform(
     bInitial=False
     Item=RigVMDispatchNode(Index=0, Array=@n1.Bones) {template="DISPATCH_RigVMDispatch_ArrayGetAtIndex(in Array,in Index,out Element)"}.Element
    ).Transform
    B=&RigUnit_GetTransform(
     bInitial=False
     Item=RigVMDispatchNode(Index=-1, Array=@n1.Bones) {template="DISPATCH_RigVMDispatch_ArrayGetAtIndex(in Array,in Index,out Element)"}.Element
    ).Transform
   ) {template="Interpolate::Execute(in A,in B,in T,out Result)"}
   DISPATCH_RigVMDispatch_If = RigVMDispatchNode(
    False=(X=0,Y=0,Z=0)
    Condition=RigVMDispatchNode(B=3, A=@DISPATCH_RigVMDispatch_ArrayGetNum.Num) {template="DISPATCH_RigVMDispatch_CoreEquals(in A,in B,out Result)"}.Result
    True=@Interpolate.Result.Translation
   ) {template="DISPATCH_RigVMDispatch_If(in Condition,in True,in False,out Result)"}
   RigVMFunction_DebugTransformMutableNoSpace = &RigVMFunction_DebugTransformMutableNoSpace(
    Color=(R=1,G=0,B=0,A=1)
    DebugDrawSettings=(DepthPriority=SDPG_Foreground,Lifetime=-1)
    Mode=Point
    Scale=10
    Thickness=1
    Transform=(Rotation=(X=0,Y=0,Z=0,W=1),Translation=(X=0,Y=0,Z=0),Scale3D=(X=1,Y=1,Z=1))
    WorldOffset=(Rotation=(X=0,Y=0,Z=0,W=1),Translation=(X=0,Y=0,Z=0),Scale3D=(X=1,Y=1,Z=1))
    Transform.Translation=@DISPATCH_RigVMDispatch_If.Result
    Transform.Rotation=@Interpolate.Result.Rotation
    ExecutePin=@RigVMFunction_Sequence_1.B
    bEnabled=RigVMVariableNode() {variable=Debug}.Value
   )
   > RigVMVariableNode(Value=@DISPATCH_RigVMDispatch_If.Result) {variable=Secondary Target}
   n3 = RigVMDispatchNode(Array=@n1.Bones, ExecuteContext=@RigVMFunction_Sequence_1.C) {template="DISPATCH_RigVMDispatch_ArrayIterator(in Array,out Element,out Index,out Count,out Ratio)"}
   Subtract = &RigVMFunction_MathIntSub(B=1, A=@n3.Count) {template="Subtract::Execute(in A,in B,out Result)"}
   DISPATCH_RigVMDispatch_CoreEquals = RigVMDispatchNode(A=@n3.Index, B=@Subtract.Result) {template="DISPATCH_RigVMDispatch_CoreEquals(in A,in B,out Result)"}
   n4 = &RigVMFunction_ControlFlowBranch(ExecuteContext=@n3.ExecuteContext, Condition=@DISPATCH_RigVMDispatch_CoreEquals.Result)
   Add = &RigVMFunction_MathIntAdd(B=1, A=@n3.Index) {template="Add::Execute(in A,in B,out Result)"}
   n5 = &RigUnit_AimBoneMath(
    DebugSettings=(bEnabled=False,Scale=10,WorldOffset=(Rotation=(X=0,Y=0,Z=0,W=1),Translation=(X=0,Y=0,Z=0),Scale3D=(X=1,Y=1,Z=1)))
    Primary=(Weight=1,Axis=(X=1,Y=0,Z=0),Target=(X=1,Y=0,Z=0),Kind=Location,Space=(Type=Bone,Name="None"))
    Secondary=(Weight=1,Axis=(X=0,Y=0,Z=1),Target=(X=1,Y=0,Z=0),Kind=Direction,Space=(Type=Bone,Name="None"))
    Weight=1
    Secondary.Kind=@n1.Kind
    Primary.Axis=@n1.Primary Axis
    Secondary.Axis=@n1.Secondary Axis
    InputTransform=&RigUnit_GetTransform(bInitial=False, Item=@n3.Element).Transform
    Primary.Target=&RigUnit_GetTransform(
     bInitial=False
     Item=RigVMDispatchNode(Index=@Add.Result, Array=@n1.Bones) {template="DISPATCH_RigVMDispatch_ArrayGetAtIndex(in Array,in Index,out Element)"}.Element
    ).Transform.Translation
    Secondary.Target=RigVMVariableNode() {variable=Secondary Target}.Value
   )
   &RigUnit_SetTransform(Weight=1, bInitial=true, Value=@n5.Result, Item=@n3.Element, ExecutePin=@n4.False)
   n6 = RigVMDispatchNode(ExecuteContext=@n3.Completed, Array=RigVMVariableNode() {variable=Bones}.Value) {template="DISPATCH_RigVMDispatch_ArrayIterator(in Array,out Element,out Index,out Count,out Ratio)"}
   DISPATCH_RigDispatch_GetMetadata = RigVMDispatchNode(Default=(Type=Bone,Name="None"), Name=Orig Parent, NameSpace=Self, Item=@n6.Element) {template="DISPATCH_RigDispatch_GetMetadata(in Item,in Name,in NameSpace,in Default,out Value,out Found)"}
   > &RigUnit_SetDefaultParent(Parent=@DISPATCH_RigDispatch_GetMetadata.Value, Child=@n6.Element)
   DISPATCH_RigVMDispatch_ArrayGetAtIndex = RigVMDispatchNode(Index=-1, Array=RigVMVariableNode() {variable=Bones}.Value) {template="DISPATCH_RigVMDispatch_ArrayGetAtIndex(in Array,in Index,out Element)"}
   Set Transform = &RigUnit_SetRotation(Value=(X=0,Y=0,Z=0,W=1), Weight=1, bInitial=true, Item=@DISPATCH_RigVMDispatch_ArrayGetAtIndex.Element, ExecutePin=@n6.Completed)
   RigVMFunction_DebugTransformArrayMutableNoSpace = /Script/RigVM.RigVMFunction_DebugTransformArrayMutableNoSpace(
    Color=(R=1,G=1,B=1,A=1)
    DebugDrawSettings=(DepthPriority=SDPG_Foreground,Lifetime=-1)
    Mode=Axes
    ParentIndices=()
    Scale=15
    Thickness=1
    WorldOffset=(Rotation=(X=0,Y=0,Z=0,W=1),Translation=(X=0,Y=0,Z=0),Scale3D=(X=1,Y=1,Z=1))
    bEnabled=@n1.Debug
    Transforms=/Script/ControlRig.RigUnit_GetTransformItemArray(Space=GlobalSpace, bInitial=False, Items=@n1.Bones) {is=struct RigVMUnitNode}.Transforms
    ExecutePin=@RigVMFunction_Sequence.B
   ) {is=struct RigVMUnitNode}
   RigVMFunctionReturnNode(ExecuteContext=@RigVMFunction_Sequence.C)
   /Script/ControlRig.RigUnit_FindItemsWithMetadataTag(NameSpace=Self, Tag=New Bones) {is=struct RigVMUnitNode}
  ]}
 New Function_ContainedGraph {nodes [
   n1 = RigVMFunctionEntryNode()
   RigVMFunction_Sequence = &RigVMFunction_Sequence(ExecuteContext=@n1.ExecuteContext)
   n2 = &RigVMFunction_Sequence(ExecuteContext=@RigVMFunction_Sequence.A)
   RigVMDispatchNode(ExecuteContext=@n2.A, Array=RigVMVariableNode() {variable=New Bones}.Value) {template="DISPATCH_RigVMDispatch_ArrayReset(io Array)"}
   > DISPATCH_RigVMDispatch_ArrayIterator = RigVMDispatchNode(Array=@n1.Transforms) {template="DISPATCH_RigVMDispatch_ArrayIterator(in Array,out Element,out Index,out Count,out Ratio)"}
   RigVMFunction_StringPadInteger = /Script/RigVM.RigVMFunction_StringPadInteger(
    Digits=2
    Value=&RigVMFunction_MathIntAdd(B=1, A=@DISPATCH_RigVMDispatch_ArrayIterator.Index) {template="Add::Execute(in A,in B,out Result)"}.Result
   ) {is=struct RigVMUnitNode}
   DISPATCH_RigDispatch_FromString = RigVMDispatchNode(String=@RigVMFunction_StringPadInteger.Result) {template="DISPATCH_RigDispatch_FromString(in String,out Result)"}
   Concat = RigVMAggregateNode(B=_, D=_, C=@DISPATCH_RigDispatch_FromString.Result, A=@n1.Name, E=@n1.Optional Suffix)
   Concat_1 = RigVMAggregateNode(B=_, D=None, C=@DISPATCH_RigDispatch_FromString.Result, A=RigVMVariableNode() {variable=Name}.Value)
   EndsWith = /Script/RigVM.RigVMFunction_EndsWith(Ending=_, Name=@Concat.Result) {is=struct RigVMUnitNode, template="EndsWith::Execute(in Name,in Ending,out Result)"}
   DISPATCH_RigVMDispatch_If = RigVMDispatchNode(False=@Concat.Result, True=@Concat_1.Result, Condition=@EndsWith.Result) {template="DISPATCH_RigVMDispatch_If(in Condition,in True,in False,out Result)"}
   > n3 = /Script/ControlRig.RigUnit_HierarchyAddBone(Space=GlobalSpace, Transform=@DISPATCH_RigVMDispatch_ArrayIterator.Element, Name=@DISPATCH_RigVMDispatch_If.Result, Parent=@n1.Parent) {is=struct RigVMUnitNode}
   > Add = RigVMDispatchNode(Element=@n3.Item, Array=RigVMVariableNode() {variable=New Bones}.Value) {template="DISPATCH_RigVMDispatch_ArrayAdd(io Array,in Element,out Index)"}
   n4 = RigVMVariableNode() {variable=New Bones}
   n5 = RigVMDispatchNode(ExecuteContext=@n2.B, Array=@n4.Value) {template="DISPATCH_RigVMDispatch_ArrayIterator(in Array,out Element,out Index,out Count,out Ratio)"}
   n6 = &RigVMFunction_MathIntSub(B=1, A=@n5.Index) {template="Subtract::Execute(in A,in B,out Result)"}
   > &RigUnit_SetDefaultParent(
    Child=@n5.Element
    Parent=RigVMDispatchNode(
     False=RigVMDispatchNode(Index=@n6.Result, Array=@n4.Value) {template="DISPATCH_RigVMDispatch_ArrayGetAtIndex(in Array,in Index,out Element)"}.Element
     Condition=RigVMDispatchNode(B=0, A=@n5.Index) {template="DISPATCH_RigVMDispatch_CoreEquals(in A,in B,out Result)"}.Result
     True=RigVMVariableNode() {variable=Parent}.Value
    ) {template="DISPATCH_RigVMDispatch_If(in Condition,in True,in False,out Result)"}.Result
   )
   RigVMFunction_ControlFlowBranch = &RigVMFunction_ControlFlowBranch(ExecuteContext=@n5.Completed, Condition=RigVMVariableNode() {variable=Enable Orient Bones}.Value)
   Orient Bones = RigVMFunctionReferenceNode(
    ExecuteContext=@RigVMFunction_ControlFlowBranch.True
    Primary Axis=RigVMVariableNode() {variable=Primary Axis}.Value
    Secondary Axis=RigVMVariableNode() {variable=Secondary Axis}.Value
    Bones=RigVMVariableNode() {variable=New Bones}.Value
    Debug=RigVMVariableNode() {variable=Debug}.Value
    Kind=RigVMVariableNode() {variable=Kind}.Value
   )
   RigVMFunctionReturnNode(ExecuteContext=@RigVMFunction_Sequence.B, Bones=RigVMVariableNode() {variable=New Bones}.Value)
   &RigUnit_FindItemsWithMetadataTag(NameSpace=Self, Tag=New Bones)
  ]}
 New Function_ContainedGraph {nodes [
   n1 = RigVMFunctionEntryNode()
   RigVMFunction_Sequence = &RigVMFunction_Sequence(ExecuteContext=@n1.ExecuteContext)
   Get_Module_Metadata = RigVMDispatchNode(Default=(Type=Bone,Name="None"), NameSpace=Root, Name=@n1.Metadata Name) {template="DISPATCH_RigDispatch_GetModuleMetadata(in Name,in NameSpace,in Default,out Value,out Found)"}
   n2 = RigVMVariableNode(Value=@Get_Module_Metadata.Found, ExecuteContext=@RigVMFunction_Sequence.A) {variable=found_metadata}
   RigVMFunction_ControlFlowBranch = &RigVMFunction_ControlFlowBranch(Condition=@Get_Module_Metadata.Found, ExecuteContext=@n2.ExecuteContext)
   n3 = /Script/ControlRig.RigUnit_HierarchyAddSocket(
    Color=@n1.Color
    Description=@n1.Description
    Transform=@n1.Extra Transform
    Name=@n1.Name
    Space=@n1.Space
    Parent=@Get_Module_Metadata.Value
    ExecutePin=@RigVMFunction_ControlFlowBranch.True
   ) {is=struct RigVMUnitNode}
   > RigVMVariableNode(Value=@n3.Item) {variable=the_socket}
   RigVMFunctionReturnNode(
    Bone=@Get_Module_Metadata.Value
    ExecuteContext=@RigVMFunction_Sequence.B
    New Socket=RigVMVariableNode() {variable=the_socket}.Value
    Found=RigVMVariableNode() {variable=found_metadata}.Value
   )
  ]}
 New Function_ContainedGraph {nodes [
   RigVMFunctionEntryNode()
   RigVMFunction_MathVectorMake = /Script/RigVM.RigVMFunction_MathVectorMake(X=0, Y=0, Z=0) {is=struct RigVMUnitNode}
   Less = &RigVMFunction_MathDoubleLess(B=0, A=@RigVMFunction_MathVectorMake.Result.X) {template="Less::Execute(in A,in B,out Result)"}
   n1 = &RigVMFunction_MathDoubleLess(B=0, A=@RigVMFunction_MathVectorMake.Result.Y) {template="Less::Execute(in A,in B,out Result)"}
   n2 = &RigVMFunction_MathDoubleLess(B=0, A=@RigVMFunction_MathVectorMake.Result.Z) {template="Less::Execute(in A,in B,out Result)"}
   RigVMFunction_MathBoolOr_3 = RigVMAggregateNode(A=@Less.Result, B=@n1.Result, C=@n2.Result)
   Greater = &RigVMFunction_MathDoubleGreater(B=0, A=@RigVMFunction_MathVectorMake.Result.X) {template="Greater::Execute(in A,in B,out Result)"}
   RigVMFunction_MathBoolOr = &RigVMFunction_MathBoolOr(A=@Greater.Result, B=@Less.Result)
   > RigVMFunctionReturnNode(
    Is Negative=RigVMDispatchNode(False=false, True=true, Condition=@RigVMFunction_MathBoolOr_3.Result) {template="DISPATCH_RigVMDispatch_If(in Condition,in True,in False,out Result)"}.Result
    X=@RigVMFunction_MathBoolOr.Result
    Y=&RigVMFunction_MathBoolOr(
     A=&RigVMFunction_MathDoubleGreater(B=0, A=@RigVMFunction_MathVectorMake.Result.Y) {template="Greater::Execute(in A,in B,out Result)"}.Result
     B=@n1.Result
    ).Result
    Z=&RigVMFunction_MathBoolOr(
     A=&RigVMFunction_MathDoubleGreater(B=0, A=@RigVMFunction_MathVectorMake.Result.Z) {template="Greater::Execute(in A,in B,out Result)"}.Result
     B=@n2.Result
    ).Result
   )
  ]}
 New Function_ContainedGraph {nodes [
   n1 = RigVMFunctionEntryNode()
   n2 = &RigUnit_CollectionNameSearchArray(PartialName=None, TypeToSearch=Bone)
   n3 = /Script/RigVM.RigVMFunction_MathBoxFromArray(Array=/Script/RigVM.RigVMFunction_MathTransformArrayToSRT(Transforms=&RigUnit_GetTransformItemArray(Space=GlobalSpace, bInitial=true, Items=@n2.Items).Transforms) {is=struct RigVMUnitNode}.Translations) {is=struct RigVMUnitNode}
   > RigVMFunctionReturnNode(
    Final Position=&RigVMFunction_MathVectorAdd(
     A=RigVMDispatchNode(True=@n3.Maximum, False=@n3.Minimum, Condition=@n1.Use Max Box Size) {template="DISPATCH_RigVMDispatch_If(in Condition,in True,in False,out Result)"}.Result
     B=@n1.Offset
    ) {template="Add::Execute(in A,in B,out Result)"}.Result
    Box=@n3.Box
   )
  ]}
 RigVMFunctionLibrary {nodes [
   Add Jiggle = RigVMCollapseNode(Child to Jiggle=(Type=Bone), Critical Damping=0.25, Ctrl=(Type=Control,Name="antenna_sec_ctrl"), ExecuteContext=(), Strength=2)
   Add Null Above = RigVMCollapseNode(ExecuteContext=(), Item=(Name="None"))
   Add Null Below = RigVMCollapseNode(
    ExecuteContext=()
    Item=(Type=None,Name="None")
    New Suffix=None
    Replace Suffix=None
    Transform=(Rotation=(X=0,Y=0,Z=0,W=1),Translation=(X=0,Y=0,Z=0),Scale3D=(X=1,Y=1,Z=1))
    Use Transform=False
   )
   AimAt = RigVMCollapseNode(Active=true, Axis=(X=1,Y=0,Z=0), ExecuteContext=(), Up Vector Axis=(X=1,Y=0,Z=0), Up Vector Object=(), object=(), target=())
   Axis to Bool = RigVMCollapseNode(Value=(X=0,Y=0,Z=0))
   ChainDynamics = RigVMCollapseNode(Bones=(), Critical Damping=1, End Control=(Type=Control,Name="None"), ExecuteContext=(), Strength=2)
   Clamp Array = RigVMCollapseNode(Array=(), ExecuteContext=(), Max=0)
   Color Module Controls = RigVMCollapseNode(CenterColor=(R=0,G=0,B=0,A=0), Controls=(), ExecuteContext=(), LeftColor=(R=0,G=0,B=0,A=0), Override=False, RightColor=(R=0,G=0,B=0,A=0))
   ColorizeControls = RigVMCollapseNode(
    CenterColor=(R=1,G=1,B=0,A=1)
    ExecuteContext=()
    LeftColor=(R=0,G=1,B=0,A=1)
    RightColor=(R=1,G=0,B=0,A=1)
    X Plane=true
    controls=()
    ignore_ctrls=()
   )
   ComputePoleVector = RigVMCollapseNode(
    Bone A=(Type=Bone)
    Bone B=(Type=Bone)
    Bone C=(Type=Bone)
    Draw Transform=true
    ExecuteContext=()
    Initial=false
    OffsetFactor=2
    Origin Scale=(X=0,Y=0,Z=0)
   )
   Connect to Module Metadata = RigVMCollapseNode(Connector=(Type=Connector,Name="Parent Control"), ExecuteContext=(), Module Metadata Name=Parent Control, Module Metadata NameSpace=Parent)
   Control Stack at Item = RigVMCollapseNode(
    Add Bottom Null=False
    Add Null=False
    Add Secondary Control=False
    Control Suffix=_ctrl
    ExecuteContext=()
    Item=(Name="None")
    Null Suffix=_null
    Parent=(Name="None")
    Secondary Settings="(Shape=(bVisible=True,Name=\"Default\",Color=(R=1,G=0,B=0,A=1),Transform=(Rotation=(X=0,Y=0,Z=0,W=1),Translation=(X=0,Y=0,Z=0),Scale3D=(X=1,Y=1,Z=1))),Proxy=(bIsProxy=False,ShapeVisibility=BasedOnSelection),DisplayName=\"None\",bSelectable=True,InitialSpace=Lo ...(truncated)"
    Settings="(Shape=(bVisible=True,Name=\"Default\",Color=(R=1,G=0,B=0,A=1),Transform=(Rotation=(X=0,Y=0,Z=0,W=1),Translation=(X=0,Y=0,Z=0),Scale3D=(X=1,Y=1,Z=1))),Proxy=(bIsProxy=False,ShapeVisibility=BasedOnSelection),DisplayName=\"None\",bSelectable=True,InitialSpace=Lo ...(truncated)"
    Tag Controls=true
    Zero Controls=true
   )
   Control Stack at Position = RigVMCollapseNode(
    Add Bottom Null=False
    Add Null=False
    Add Secondary Control=False
    Control Suffix=_ctrl
    ExecuteContext=()
    Global Position=(Rotation=(X=0,Y=0,Z=0,W=1),Translation=(X=0,Y=0,Z=0),Scale3D=(X=1,Y=1,Z=1))
    Null Suffix=_null
    Parent=(Type=None,Name="None")
    Secondary Settings="(Shape=(bVisible=True,Name=\"Default\",Color=(R=1,G=0,B=0,A=1),Transform=(Rotation=(X=0,Y=0,Z=0,W=1),Translation=(X=0,Y=0,Z=0),Scale3D=(X=1,Y=1,Z=1))),Proxy=(bIsProxy=False,ShapeVisibility=BasedOnSelection),DisplayName=\"None\",bSelectable=True,InitialSpace=Lo ...(truncated)"
    Settings="(Shape=(bVisible=True,Name=\"Default\",Color=(R=1,G=0,B=0,A=1),Transform=(Rotation=(X=0,Y=0,Z=0,W=1),Translation=(X=0,Y=0,Z=0),Scale3D=(X=1,Y=1,Z=1))),Proxy=(bIsProxy=False,ShapeVisibility=BasedOnSelection),DisplayName=\"None\",bSelectable=True,InitialSpace=Lo ...(truncated)"
    Tag Controls=true
    Zero Controls=true
   )
   Create Bone Chain = RigVMCollapseNode(
    Debug=false
    Debug Scale=15
    Debug Thickness=0.5
    Enable Orient Bones=true
    ExecuteContext=()
    Kind=Direction
    Name=NewBone
    Parent=(Type=Bone,Name="None")
    Primary Axis=(X=1,Y=0,Z=0)
    Secondary Axis=(X=0,Y=1,Z=0)
    Transforms=()
   )
   Create Slope Rectangle = RigVMCollapseNode(
    Axis=(X=0,Y=0,Z=1)
    Color=(R=1,G=0,B=1,A=1)
    ExecuteContext=()
    Hit Radius=50
    Item=(Type=Bone,Name="None")
    Rectangle Size=250
    Rectangle Thickness=2
   )
   Create Socket at Metadata = RigVMCollapseNode(
    Color=(R=1,G=1,B=1,A=1)
    Description=Place the shoulder module here
    ExecuteContext=()
    Extra Transform=(Rotation=(X=0,Y=0,Z=0,W=1),Translation=(X=0,Y=0,Z=0),Scale3D=(X=1,Y=1,Z=1))
    Metadata Name=None
    Name=shoulder_l
    Space=LocalSpace
   )
   Debug Item = RigVMCollapseNode(Active=true, ExecuteContext=(), Item=(Type=Bone), Scale=15, Thickness=2)
   Debug Items = RigVMCollapseNode(Active=true, ExecuteContext=(), Items=(), Scale=10)
   Debug Transforms = RigVMCollapseNode(Active=true, Color=(R=1,G=0,B=0,A=1), ExecuteContext=(), Mode=Axes, Scale=10, Transforms=())
   DrawAxis = RigVMCollapseNode(Active=true, ExecuteContext=(), Item=(Type=Bone), Scale=15, Thickness=2)
   DrawItems = RigVMCollapseNode(Active=true, ExecuteContext=(), Items=(), Scale=10)
   DrawSlopeRectangle = RigVMCollapseNode(
    Axis=(X=0,Y=0,Z=1)
    Color=(R=1,G=0,B=1,A=1)
    ExecuteContext=()
    Hit Radius=50
    Item=(Type=Bone,Name="None")
    Rectangle Size=250
    Rectangle Thickness=2
   )
   DrawTransforms = RigVMCollapseNode(Active=true, Color=(R=1,G=0,B=0,A=1), ExecuteContext=(), Mode=Axes, Scale=10, Transforms=())
   FKChain = RigVMCollapseNode(Bones=(), Controls=(), ExecuteContext=(), Rotation Filter=(bX=True,bY=True,bZ=True), Translation Filter=(bX=True,bY=True,bZ=True))
   Find first in Hierarchy = RigVMCollapseNode(Children=true, ExecuteContext=(), Ignore=(), Item=(Type=Bone,Name="None"), Search for=(), Type to Search=Bone)
   FindClosestBone = RigVMCollapseNode(Bones=(), ExecuteContext=(), Initial=true, Item=())
   Get BBox of Skeleton = RigVMCollapseNode(Offset=(X=0,Y=0,Z=50), Use Max Box Size=true)
   Get Controls = RigVMCollapseNode()
   GetLengthOfChain = RigVMCollapseNode(
    Bones=((Type=Bone,Name="torso_1"),(Type=Bone,Name="torso_2"),(Type=Bone,Name="torso_3"),(Type=Bone,Name="torso_4"))
    ExecuteContext=()
    Initial=true
   )
   GetLengthOfChain_57 = RigVMCollapseNode(
    Bones=((Type=Bone,Name="torso_1"),(Type=Bone,Name="torso_2"),(Type=Bone,Name="torso_3"),(Type=Bone,Name="torso_4"))
    ExecuteContext=()
    Initial=true
    Initial_1=(Pitch=0,Yaw=0,Roll=0)
   )
   GetTransforms = RigVMCollapseNode(ExecuteContext=(), Items=())
   HideControls = RigVMCollapseNode(Active=true, ExecuteContext=(), controls=())
   IKThreeBone = RigVMCollapseNode(
    Bone A=(Type=Bone,Name="None")
    Bone B=(Type=Bone,Name="None")
    Bone C=(Type=Bone,Name="None")
    Bone D=(Type=Control,Name="None")
    Enable Stretch=true
    ExecuteContext=()
    Foot Ctrl=(Type=Control,Name="None")
    Pole Vector Node=(Type=Control,Name="None")
    Primary Axis=(X=1,Y=0,Z=0)
    Secondary Axis=(X=0,Y=-1,Z=0)
    Stretch Maximum Ratio=1.25
    Thigh Ctrl Optional=()
    scale=(X=0,Y=0,Z=0)
   )
   IKThreeBone v2 = RigVMCollapseNode(
    Bone A=(Type=Bone,Name="None")
    Bone B=(Type=Bone,Name="None")
    Bone C=(Type=Bone,Name="None")
    Bone D=(Type=Control,Name="None")
    Enable Stretch=true
    ExecuteContext=()
    Foot Ctrl=(Type=Control,Name="None")
    Pole Vector Node=(Type=Control,Name="None")
    Primary Axis=(X=1,Y=0,Z=0)
    Secondary Axis=(X=0,Y=-1,Z=0)
    Stretch Maximum Ratio=1.25
    Thigh Ctrl Optional=()
    scale=(X=0,Y=0,Z=0)
   )
   IKTwoBone = RigVMCollapseNode(
    Bone A=(Type=Bone,Name="None")
    Bone B=(Type=Bone,Name="None")
    Bone C=(Type=Bone,Name="None")
    Enable Stretch=false
    End Ctrl=(Type=Control,Name="None")
    ExecuteContext=()
    Line Thickness=0.25
    Pole Vector Node=(Type=Control,Name="None")
    Primary Axis=(X=1,Y=0,Z=0)
    Secondary Axis=(X=0,Y=-1,Z=0)
    Stretch Maximum Ratio=1.25
    Stretch Start Ratio=0.75
    debug=true
    scale=(X=0,Y=0,Z=0)
   )
   Local Chain Dynamics = RigVMCollapseNode(Blend End=1, Blend Start=1, Blend Twist=1, Bones=(), Damping=0.5, ExecuteContext=(), Strength=4)
   Local Chain Dynamics v02 = RigVMCollapseNode(
    Blend End=1
    Blend Start=1
    Blend Twist=1
    Bones=()
    Damping=0.5
    Draw Debug=false
    ExecuteContext=()
    Force=(X=0,Y=0,Z=0)
    Primary Axis=(X=1,Y=0,Z=0)
    Secondary Axis=(X=0,Y=1,Z=0)
    Strength=4
   )
   MirrorControl = RigVMCollapseNode(
    ExecuteContext=()
    MirrorAxis=(X=-1,Y=1,Z=1)
    MirrorFrom=(Type=Bone)
    MirrorTo=(Type=Control,Name="foot_roll_inner_r_ctrl")
    Propagate to Children=false
   )
   Orient Bones = RigVMCollapseNode(Bones=(), Debug=false, ExecuteContext=(), Kind=Direction, Primary Axis=(X=1,Y=0,Z=0), Secondary Axis=(X=0,Y=1,Z=0))
   Parent = RigVMCollapseNode(Child=(Type=Bone,Name="None"), ExecuteContext=(), New Parent=(Type=Bone,Name="None"))
   Piston = RigVMCollapseNode(
    Axis=(X=1,Y=0,Z=0)
    Debug Draw=true
    End Bone=()
    End Ctrl=()
    ExecuteContext=()
    Maintain Offset=true
    Start Bone=(Type=Bone)
    Start Ctrl=()
    Up Ctrl=(Type=Control)
    Up Vector Axis=(X=0,Y=0,Z=1)
   )
   Project Controls to New Parent = RigVMCollapseNode(Array=(), ExecuteContext=(), New Parent=(Type=Control,Name="jaw_ctrl"))
   ReparentControls = RigVMCollapseNode(Ctrls=(), ExecuteContext=(), IK Parent=(Type=Control,Name="None"))
   Scale Module Controls = RigVMCollapseNode(Controls=(), ExecuteContext=(), Global Scale=1, Local Scale=0)
   Set Offset Transform from Current = RigVMCollapseNode(Control=(Type=Bone,Name="None"), ExecuteContext=())
   SnapTransform = RigVMCollapseNode(Active=true, ExecuteContext=(), Item=(Type=Bone), Propagate to Children=True, Space=GlobalSpace, Target=(Type=Bone))
   TwistBones = RigVMCollapseNode(Axis to Twist=(bX=True,bY=false,bZ=false), Bones to Twist=(), End=(Type=Bone,Name="None"), ExecuteContext=(), Start=(Type=Bone,Name="None"))
  ]}
 RigVMFunction_MathBoolAnd_1_1_ContainedGraph {nodes [
   n1 = RigVMFunctionEntryNode()
   RigVMFunctionReturnNode(Result=&RigVMFunction_MathBoolAnd(B=@n1.C, A=&RigVMFunction_MathBoolAnd(A=@n1.A, B=@n1.B).Result).Result)
  ]}
 RigVMFunction_MathBoolOr_3_1_ContainedGraph {nodes [
   n1 = RigVMFunctionEntryNode()
   RigVMFunctionReturnNode(Result=&RigVMFunction_MathBoolOr(B=@n1.C, A=&RigVMFunction_MathBoolOr(A=@n1.A, B=@n1.B).Result).Result)
  ]}
 RigVMFunction_MathBoolOr_3_1_ContainedGraph {nodes [
   n1 = RigVMFunctionEntryNode()
   RigVMFunctionReturnNode(Result=&RigVMFunction_MathBoolOr(B=@n1.C, A=&RigVMFunction_MathBoolOr(A=@n1.A, B=@n1.B).Result).Result)
  ]}
 RigVMFunction_Sequence_1_1_ContainedGraph {nodes [
   n1 = RigVMFunctionEntryNode()
   n2 = &RigVMFunction_Sequence(ExecuteContext=@n1.ExecuteContext)
   n3 = &RigVMFunction_Sequence(ExecuteContext=@n2.B)
   n4 = &RigVMFunction_Sequence(ExecuteContext=@n3.B)
   RigVMFunctionReturnNode(A=@n2.A, B=@n3.A, C=@n4.A, D=@n4.B)
  ]}
 RigVMFunction_Sequence_1_ContainedGraph {nodes [
   n1 = RigVMFunctionEntryNode()
   RigVMFunction_Sequence = &RigVMFunction_Sequence(ExecuteContext=@n1.ExecuteContext)
   n2 = &RigVMFunction_Sequence(ExecuteContext=@RigVMFunction_Sequence.B)
   RigVMFunctionReturnNode(A=@RigVMFunction_Sequence.A, B=@n2.A, C=@n2.B)
  ]}
 RigVMFunction_Sequence_2_ContainedGraph {nodes [
   n1 = RigVMFunctionEntryNode()
   RigVMFunction_Sequence = &RigVMFunction_Sequence(ExecuteContext=@n1.ExecuteContext)
   n2 = &RigVMFunction_Sequence(ExecuteContext=@RigVMFunction_Sequence.B)
   RigVMFunctionReturnNode(A=@RigVMFunction_Sequence.A, B=@n2.A, C=@n2.B)
  ]}
 RigVMModel {nodes [&RigUnit_BeginExecution()]}
]
